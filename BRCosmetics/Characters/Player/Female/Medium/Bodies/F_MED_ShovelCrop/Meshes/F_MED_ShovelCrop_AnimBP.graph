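# Animation Blueprint for the F_MED_ShovelCrop female character body, following the parent mesh pose with added secondary motion.
# It layers back-hair bones from FaceMesh, runs a Control Rig, and offsets backpack and cape attach bones, rolling the backpack while skydiving.
# The cached result feeds chained angular and linear spring AnimDynamics on pec and glute bones.
# Events toggle body and shirt morphs above LOD 0 and set a material element-hiding parameter.
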AnimBlueprint /BRCosmetics/Characters/Player/Female/Medium/Bodies/F_MED_ShovelCrop/Meshes/F_MED_ShovelCrop_AnimBP
target_skeleton=&/F_MED_ShovelCrop_Skeleton
graphs [
 AnimGraph {nodes [
   Blend Poses by bool(BlendPose_1=Use cached pose 'Predynamics'(), BlendTime_0=0, BlendTime_1=0, bActiveValue=Get bPbBlendEnabled() -> (bPbBlendEnabled:bool))
   n1 = Get Sim Space Settings() -> (Sim Space Settings:Anim Phys Sim Space Settings Structure)
   n2 = Get Sim Space Settings()
   n3 = Anim Dynamics - Bone: dyn_pec_r!(
    ComponentPose=Local To Component(LocalPose=Use cached pose 'Predynamics'())
    Alpha=1
    AngularSpringConstant:float=5000
    SimSpaceSettings:Anim Phys Sim Space Settings Structure=@n2
   ) -> (Pose) {
    angularBiasOverride=1
    angularDampingOverride=0.9999
    angularSpringConstant=5000
    bAngularSpring=true
    bGravityOverrideInSimSpace=true
    bOverrideAngularBias=true
    bOverrideAngularDamping=true
    bUseGravityOverride=true
    linearDampingOverride=0.9999
    linearSpringConstant=100
    simulationSpace=RootRelative
    boundBone {boneName=dyn_pec_r}
    componentAppliedLinearAccClamp=<2 2 2>
    gravityOverride {z=-980}
    physicsBodyDefinitions [{
      boundBone {boneName=dyn_pec_r}
      boxExtents=<21 21 21>
      constraintSetup {angularLimitsMax=<0 1.5 6.5>, angularLimitsMin=<0 -6.5 -6.5>, angularTarget=<1 0 0>}
      localJointOffset=<2 0 0>
     }]
    relativeSpaceBone {boneName=spine_04}
    simSpaceSettings {maxAngularAcceleration=45, maxAngularVelocity=8, simSpaceAngularAlpha=0.25}
   }
   n4 = Anim Dynamics - Bone: dyn_pec_l!(ComponentPose=@n3, Alpha=1, AngularSpringConstant:float=5000, SimSpaceSettings:Anim Phys Sim Space Settings Structure=@n2) -> (Pose) {
    angularBiasOverride=1
    angularDampingOverride=0.9999
    angularSpringConstant=5000
    bAngularSpring=true
    bGravityOverrideInSimSpace=true
    bOverrideAngularBias=true
    bOverrideAngularDamping=true
    bUseGravityOverride=true
    linearDampingOverride=0.9999
    linearSpringConstant=100
    simulationSpace=RootRelative
    boundBone {boneName=dyn_pec_l}
    componentAppliedLinearAccClamp=<2 2 2>
    gravityOverride {z=-980}
    physicsBodyDefinitions [{
      boundBone {boneName=dyn_pec_l}
      boxExtents=<20 20 20>
      constraintSetup {angularLimitsMax=<0 1.5 6.5>, angularLimitsMin=<0 -6.5 -6.5>, angularTarget=<1 0 0>}
      localJointOffset=<2 0 0>
     }]
    relativeSpaceBone {boneName=spine_04}
    simSpaceSettings {maxAngularAcceleration=45, maxAngularVelocity=8, simSpaceAngularAlpha=0.25}
   }
   n5 = Anim Dynamics - Bone: dyn_pec_r!(ComponentPose=@n4, Alpha=1, GravityOverride:Vector="0,0,-200", LinearSpringConstant:float=180) -> (Pose) {
    angularBiasOverride=1
    angularDampingOverride=0.99
    angularSpringConstant=10000
    bGravityOverrideInSimSpace=true
    bLinearSpring=true
    bOverrideAngularBias=true
    bOverrideLinearDamping=true
    bUseGravityOverride=true
    linearDampingOverride=1
    linearSpringConstant=180
    simulationSpace=RootRelative
    boundBone {boneName=dyn_pec_r}
    componentAppliedLinearAccClamp=<5 5 5>
    gravityOverride {z=-200}
    physicsBodyDefinitions [{
      boundBone {boneName=dyn_pec_r}
      boxExtents=<80 80 80>
      constraintSetup {angularLimitsMax=<0 0 0>, angularLimitsMin=<0 0 0>, angularTarget=<1 0 0>, linearAxesMax=<0 0.7 1.2>, linearAxesMin=<0 -0.7 -0.7>}
      localJointOffset=<0.2 0 0>
     }]
    relativeSpaceBone {boneName=spine_04}
    simSpaceSettings {maxAngularAcceleration=50, maxAngularVelocity=10, simSpaceAngularAlpha=0.1}
   }
   n6 = Anim Dynamics - Bone: dyn_pec_l!(ComponentPose=@n5, Alpha=1, GravityOverride:Vector="0,0,-200", LinearSpringConstant:float=180) -> (Pose) {
    angularBiasOverride=1
    angularDampingOverride=0.99
    angularSpringConstant=10000
    bGravityOverrideInSimSpace=true
    bLinearSpring=true
    bOverrideAngularBias=true
    bOverrideLinearDamping=true
    bUseGravityOverride=true
    linearDampingOverride=1
    linearSpringConstant=180
    simulationSpace=RootRelative
    boundBone {boneName=dyn_pec_l}
    componentAppliedLinearAccClamp=<5 5 5>
    gravityOverride {z=-200}
    physicsBodyDefinitions [{
      boundBone {boneName=dyn_pec_l}
      boxExtents=<90 90 90>
      constraintSetup {angularLimitsMax=<0 0 0>, angularLimitsMin=<0 0 0>, angularTarget=<1 0 0>, linearAxesMax=<0 0.7 1.2>, linearAxesMin=<0 -0.7 -0.7>}
      localJointOffset=<0.2 0 0>
     }]
    relativeSpaceBone {boneName=spine_04}
    simSpaceSettings {maxAngularAcceleration=50, maxAngularVelocity=10, simSpaceAngularAlpha=0.1}
   }
   n7 = Anim Dynamics - Bone: deform_glute_r!(ComponentPose=@n6, Alpha=1, AngularSpringConstant:float=5000, SimSpaceSettings:Anim Phys Sim Space Settings Structure=@n1) -> (Pose) {
    angularBiasOverride=1
    angularDampingOverride=0.9999
    angularSpringConstant=5000
    bAngularSpring=true
    bGravityOverrideInSimSpace=true
    bOverrideAngularBias=true
    bOverrideAngularDamping=true
    bUseGravityOverride=true
    linearDampingOverride=0.9999
    linearSpringConstant=100
    simulationSpace=RootRelative
    boundBone {boneName=deform_glute_r}
    componentAppliedLinearAccClamp=<2 2 2>
    gravityOverride {z=-980}
    physicsBodyDefinitions [{
      boundBone {boneName=deform_glute_r}
      boxExtents=<21 21 21>
      constraintSetup {angularLimitsMax=<0 2 2>, angularLimitsMin=<0 -2 -2>, angularTarget=<1 0 0>}
      localJointOffset=<2 0 0>
     }]
    relativeSpaceBone {boneName=spine_04}
    simSpaceSettings {maxAngularAcceleration=45, maxAngularVelocity=8, simSpaceAngularAlpha=0.25}
   }
   n8 = Anim Dynamics - Bone: deform_glute_l!(ComponentPose=@n7, Alpha=1, AngularSpringConstant:float=5000, SimSpaceSettings:Anim Phys Sim Space Settings Structure=@n1) -> (Pose) {
    angularBiasOverride=1
    angularDampingOverride=0.9999
    angularSpringConstant=5000
    bAngularSpring=true
    bGravityOverrideInSimSpace=true
    bOverrideAngularBias=true
    bOverrideAngularDamping=true
    bUseGravityOverride=true
    linearDampingOverride=0.9999
    linearSpringConstant=100
    simulationSpace=RootRelative
    boundBone {boneName=deform_glute_l}
    componentAppliedLinearAccClamp=<2 2 2>
    gravityOverride {z=-980}
    physicsBodyDefinitions [{
      boundBone {boneName=deform_glute_l}
      boxExtents=<20 20 20>
      constraintSetup {angularLimitsMax=<0 2 2>, angularLimitsMin=<0 -2 -2>, angularTarget=<1 0 0>}
      localJointOffset=<2 0 0>
     }]
    relativeSpaceBone {boneName=spine_04}
    simSpaceSettings {maxAngularAcceleration=45, maxAngularVelocity=8, simSpaceAngularAlpha=0.25}
   }
   n9 = Anim Dynamics - Bone: deform_glute_r!(ComponentPose=@n8, Alpha=1, GravityOverride:Vector="0,0,-200", LinearSpringConstant:float=180) -> (Pose) {
    angularBiasOverride=1
    angularDampingOverride=0.99
    angularSpringConstant=10000
    bGravityOverrideInSimSpace=true
    bLinearSpring=true
    bOverrideAngularBias=true
    bOverrideLinearDamping=true
    bUseGravityOverride=true
    linearDampingOverride=1
    linearSpringConstant=180
    simulationSpace=RootRelative
    boundBone {boneName=deform_glute_r}
    componentAppliedLinearAccClamp=<5 5 5>
    gravityOverride {z=-200}
    physicsBodyDefinitions [{
      boundBone {boneName=deform_glute_r}
      boxExtents=<80 80 80>
      constraintSetup {angularLimitsMax=<0 0 0>, angularLimitsMin=<0 0 0>, angularTarget=<1 0 0>, linearAxesMax=<0 2 2>, linearAxesMin=<0 -2 -2>}
      localJointOffset=<0.2 0 0>
     }]
    relativeSpaceBone {boneName=spine_04}
    simSpaceSettings {maxAngularAcceleration=50, maxAngularVelocity=10, simSpaceAngularAlpha=0.1}
   }
   n10 = Anim Dynamics - Bone: deform_glute_l!(ComponentPose=@n9, Alpha=1, GravityOverride:Vector="0,0,-200", LinearSpringConstant:float=180) -> (Pose) {
    angularBiasOverride=1
    angularDampingOverride=0.99
    angularSpringConstant=10000
    bGravityOverrideInSimSpace=true
    bLinearSpring=true
    bOverrideAngularBias=true
    bOverrideLinearDamping=true
    bUseGravityOverride=true
    linearDampingOverride=1
    linearSpringConstant=180
    simulationSpace=RootRelative
    boundBone {boneName=deform_glute_l}
    componentAppliedLinearAccClamp=<5 5 5>
    gravityOverride {z=-200}
    physicsBodyDefinitions [{
      boundBone {boneName=deform_glute_l}
      boxExtents=<90 90 90>
      constraintSetup {angularLimitsMax=<0 0 0>, angularLimitsMin=<0 0 0>, angularTarget=<1 0 0>, linearAxesMax=<0 2 2>, linearAxesMin=<0 -2 -2>}
      localJointOffset=<0.2 0 0>
     }]
    relativeSpaceBone {boneName=spine_04}
    simSpaceSettings {maxAngularAcceleration=50, maxAngularVelocity=10, simSpaceAngularAlpha=0.1}
   }
   Output Pose(Result=Component To Local(ComponentPose=@n10)) {layerGroup=None}
   n11 = Layered blend per bone(
    BasePose=Copy Pose From Mesh() {bUseAttachedParent=true}
    BlendPoses_0=Copy Pose From Mesh(SourceMeshComponent=Get FaceMesh() -> (FaceMesh:Skeletal Mesh Component Budgeted Object Reference)) {bUseAttachedParent=true}
   ) {layerSetup [{branchFilters [
       {boneName=base_hair_back_00}
       {boneName=dyn_hair_back_00}
       {boneName=dyn_hair_back_01}
       {boneName=dyn_hair_back_02}
       {boneName=dyn_hair_back_03}
       {boneName=dyn_hair_back_04}
       {boneName=dyn_hair_back_05}
       {boneName=aim_hair_back_01}
       {boneName=offset_hair_back_01}
       {boneName=base_hair_back_00_d}
       {boneName=base_hair_back_00_e}
       {boneName=dyn_hair_back_00_d}
       {boneName=dyn_hair_back_00_e}
       {boneName=dyn_hair_back_01_d}
       {boneName=dyn_hair_back_01_e}
       {boneName=base_hair_MSTR}
      ]}]}
   n12 = Control Rig(Source=@n11) {controlRigAssetReference {controlRigAsset=&/F_MED_ShovelCrop_CtrlRig}}
   n13 = "Transform (Modify) Bone - Bone: attach_backpack"(ComponentPose=Local To Component(LocalPose=@n12), Alpha=1, Translation="0,-8,0", Rotation="0,0,0", Scale="1,1,1") {translationMode=BMM_Additive, translationSpace=BCS_BoneSpace, boneToModify {boneName=attach_backpack}, translation {y=-8}}
   n14 = "Transform (Modify) Bone - Bone: attach_backpack"!(ComponentPose=@n13, bAlphaBoolEnabled:bool=Get isSkydiving() -> (isSkydiving:bool), Translation="0,0,0", Rotation="0,0,30", Scale="1,1,1") -> (Pose) {alphaInputType=Bool, rotationMode=BMM_Additive, rotationSpace=BCS_BoneSpace, boneToModify {boneName=attach_backpack}, rotation {roll=30}}
   n15 = "Transform (Modify) Bone - Bone: attach_cape"(ComponentPose=@n14, Alpha=1, Translation="0,0.5,-1", Rotation="0,0,0", Scale="1,1,1") {translationMode=BMM_Additive, translationSpace=BCS_BoneSpace, boneToModify {boneName=attach_cape}, translation {y=0.5, z=-1}}
   Save cached pose 'Predynamics'(Pose=Component To Local(ComponentPose=@n15)) {cacheName=Predynamics}
  ]}
 EventGraph {nodes [
   n1 = Event OnWeaponChanged()
   n2 = Get Dynamic Material Instance(execute=[@n1.then, @n4.then], SkeletalMeshComponent=GetPartSkeletalMeshComponent())
   > n3 = Branch(Condition=false)
   > SetScalarParameterValue(self=@n2, ParameterName=Hide Element 03, Value=1)
   SetScalarParameterValue(execute=@n3.else, self=@n2, ParameterName=Hide Element 03, Value=0)
   n4 = Event OnCharacterCustomizationLoaded()
   Event OnBeginSkydiving()
   > Set isSkydiving(isSkydiving:bool=true) -> (Output_Get:bool)
   Event BlueprintUpdateAnimation()
   > n5 = Branch(Condition="integer > integer"(A=Get CurrentLOD(), B=0) {is=title K2Node_CallFunction})
   > SetMorphTarget!(execute, self:Anim Instance Object Reference, MorphTargetName:Name=Body_BS, Value:float=1) -> (then)
   > SetMorphTarget(MorphTargetName=Shirt_BS, Value=1)
   SetMorphTarget(execute=@n5.else, MorphTargetName=Body_BS, Value=0)
   > SetMorphTarget(MorphTargetName=Shirt_BS, Value=0)
   Event OnEndSkydiving()
   > Set isSkydiving(isSkydiving=false)
   n6 = Set FaceMesh(FaceMesh:Skeletal Mesh Component Budgeted Object Reference=Get Array Item with EFort Custom Part Type(Array=@n4.CustomMeshParts, PartType=Face)) -> (Output_Get:Skeletal Mesh Component Budgeted Object Reference)
   > AddTickPrerequisiteComponent!(execute, self:Actor Component Object Reference=GetOwningComponent(), PrerequisiteComponent:Actor Component Object Reference=@n6) -> (then)
   PickaxeItemDefinitionHasAnyMetaTag(Tags=Make Array("[0]"=(TagName="Cosmetics.MetaTags.BackSpin")), bRequiresEquipped=true)
  ]}
]
variables [
 FaceMesh {type=SkeletalMeshComponentBudgeted}
 isSkydiving {type=boolean}
 Sim Space Settings {
  type=AnimPhysSimSpaceSettings
  default {maxAngularAcceleration=45, maxAngularVelocity=8, simSpaceAngularAlpha=1, externalAngularVelocity=<0 0 0>}
 }
]
