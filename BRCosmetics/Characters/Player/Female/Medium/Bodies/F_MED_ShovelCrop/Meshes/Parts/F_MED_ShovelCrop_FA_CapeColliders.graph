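# Physics asset named as cape colliders for a part of the female medium ShovelCrop player cosmetic body.
# It defines twelve bodies on dynamic back-hair bones (dyn_hair_back), spanning numbered segments 01 to 05 and several lettered variants.
# It has no constraints between bodies, so it supplies no joint setup of its own.
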PhysicsAsset /BRCosmetics/Characters/Player/Female/Medium/Bodies/F_MED_ShovelCrop/Meshes/Parts/F_MED_ShovelCrop_FA_CapeColliders
bodies {
 dyn_hair_back_01_a []
 dyn_hair_back_01_c []
 dyn_hair_back_01_d []
 dyn_hair_back_01_e []
 dyn_hair_back_02 []
 dyn_hair_back_02_a []
 dyn_hair_back_02_d []
 dyn_hair_back_02_e []
 dyn_hair_back_03 []
 dyn_hair_back_03_d []
 dyn_hair_back_04 []
 dyn_hair_back_05 []
}
constraints []
