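# Control Rig for the ShovelCrop female face accessory that poses back hair to follow head and shoulder motion.
# Spherical pose readers on head and clavicles feed weights that shift an aim target, with a separate cape branch and smoothed crouch adjustments.
# An aim constraint turns the back hair bone toward that target.
# An offset bone below it adds corrective rotations from clavicles, movement, cape and rigid-body state.
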ControlRigBlueprint /BRCosmetics/Characters/Player/Female/Medium/Bodies/F_MED_ShovelCrop/Meshes/Parts/F_MED_ShovelCrop_FaceAcc_CtrlRig
graphs [
 CollapseNode_ContainedGraph {nodes [
   n1 = RigVMFunctionEntryNode()
   n2 = &RigVMFunction_Sequence(ExecuteContext=@n1.ExecuteContext)
   RigVMFunction_Sequence_2 = RigVMAggregateNode(ExecuteContext=@n2.A)
   n3 = &RigUnit_SphericalPoseReader(
    ActiveRegionScaleFactors=(PositiveWidth=1.35,NegativeWidth=1.35,PositiveHeight=0.45,NegativeHeight=1.9)
    ActiveRegionSize=0.1
    Debug=(bDrawDebug=false,bDraw2D=False,bDrawLocalAxes=False,DebugScale=25,DebugSegments=20,DebugThickness=0.25)
    DriverAxis=(X=1,Y=1,Z=0)
    DriverItem=(Type=Bone,Name="head")
    FalloffRegionScaleFactors=(PositiveWidth=1,NegativeWidth=1,PositiveHeight=1.0266,NegativeHeight=1)
    FalloffSize=0.235
    FlipHeightScaling=False
    FlipWidthScaling=False
    OptionalParentItem=(Type=Bone,Name="None")
    RotationOffset=(X=0,Y=-90,Z=0)
    ExecutePin=@RigVMFunction_Sequence_2.A
   )
   > RigVMVariableNode(Value=@n3.OutputParam) {variable=head_up}
   > n4 = &RigUnit_SphericalPoseReader(
    ActiveRegionScaleFactors=(PositiveWidth=1.35,NegativeWidth=1.35,PositiveHeight=1,NegativeHeight=1.9)
    ActiveRegionSize=0.1968
    Debug=(bDrawDebug=false,bDraw2D=False,bDrawLocalAxes=False,DebugScale=25,DebugSegments=20,DebugThickness=0.25)
    DriverAxis=(X=1,Y=1,Z=0)
    DriverItem=(Type=Bone,Name="head")
    FalloffRegionScaleFactors=(PositiveWidth=1,NegativeWidth=1,PositiveHeight=1.0266,NegativeHeight=1)
    FalloffSize=0.235
    FlipHeightScaling=False
    FlipWidthScaling=False
    OptionalParentItem=(Type=Bone,Name="None")
    RotationOffset=(X=-37,Y=-90,Z=0)
   )
   > RigVMVariableNode(Value=@n4.OutputParam) {variable=head_overshoot}
   > n5 = &RigUnit_SphericalPoseReader(
    ActiveRegionScaleFactors=(PositiveWidth=0.21,NegativeWidth=1,PositiveHeight=1,NegativeHeight=1)
    ActiveRegionSize=0.5
    Debug=(bDrawDebug=false,bDraw2D=false,bDrawLocalAxes=False,DebugScale=25,DebugSegments=20,DebugThickness=0.25)
    DriverAxis=(X=1,Y=0,Z=0)
    DriverItem=(Type=Bone,Name="head")
    FalloffRegionScaleFactors=(PositiveWidth=1,NegativeWidth=1,PositiveHeight=1,NegativeHeight=1)
    FalloffSize=0.5345
    FlipHeightScaling=False
    FlipWidthScaling=False
    OptionalParentItem=(Type=Bone,Name="None")
    RotationOffset=(X=90,Y=0,Z=6.2359)
   )
   > RigVMVariableNode(Value=@n5.OutputParam) {variable=head_down}
   > n6 = &RigUnit_SphericalPoseReader(
    ActiveRegionScaleFactors=(PositiveWidth=1,NegativeWidth=1,PositiveHeight=1,NegativeHeight=1)
    ActiveRegionSize=0.15
    Debug=(bDrawDebug=false,bDraw2D=false,DebugScale=25,DebugSegments=20,DebugThickness=0.25)
    DriverAxis=(X=0,Y=1,Z=0)
    DriverItem=(Type=Bone,Name="head")
    FalloffRegionScaleFactors=(PositiveWidth=1,NegativeWidth=1,PositiveHeight=1,NegativeHeight=1)
    FalloffSize=0.32
    OptionalParentItem=(Type=Bone)
    RotationOffset=(X=150,Y=0,Z=0)
   )
   > RigVMVariableNode(Value=@n6.OutputParam) {variable=head_l}
   > n7 = &RigUnit_SphericalPoseReader(
    ActiveRegionScaleFactors=(PositiveWidth=1,NegativeWidth=1,PositiveHeight=1,NegativeHeight=1)
    ActiveRegionSize=0.15
    Debug=(bDrawDebug=false,bDraw2D=false,DebugScale=25,DebugSegments=20,DebugThickness=0.25)
    DriverAxis=(X=0,Y=1,Z=0)
    DriverItem=(Type=Bone,Name="head")
    FalloffRegionScaleFactors=(PositiveWidth=1,NegativeWidth=1,PositiveHeight=1,NegativeHeight=1)
    FalloffSize=0.3
    FlipHeightScaling=true
    FlipWidthScaling=false
    OptionalParentItem=(Type=Bone)
    RotationOffset=(X=30,Y=0,Z=0)
   )
   > RigVMVariableNode(Value=@n7.OutputParam) {variable=head_r}
   > n8 = &RigUnit_SphericalPoseReader(
    ActiveRegionScaleFactors=(PositiveWidth=3,NegativeWidth=1,PositiveHeight=3,NegativeHeight=1)
    ActiveRegionSize=0.1
    Debug=(bDrawDebug=true,bDraw2D=False,bDrawLocalAxes=False,DebugScale=25,DebugSegments=20,DebugThickness=0.25)
    DriverAxis=(X=-1,Y=0,Z=0)
    DriverItem=(Type=Bone,Name="clavicle_r")
    FalloffRegionScaleFactors=(PositiveWidth=-1,NegativeWidth=1,PositiveHeight=-1,NegativeHeight=1)
    FalloffSize=0.2
    FlipHeightScaling=False
    FlipWidthScaling=False
    OptionalParentItem=(Type=Bone,Name="None")
    RotationOffset=(X=110,Y=-15,Z=51)
   )
   > RigVMVariableNode(Value=@n8.OutputParam) {variable=clavicleback_r}
   > n9 = &RigUnit_SphericalPoseReader(
    ActiveRegionScaleFactors=(PositiveWidth=1,NegativeWidth=1,PositiveHeight=3,NegativeHeight=1)
    ActiveRegionSize=0.1
    Debug=(bDrawDebug=false,bDraw2D=False,bDrawLocalAxes=False,DebugScale=25,DebugSegments=20,DebugThickness=0.25)
    DriverAxis=(X=1,Y=0,Z=0)
    DriverItem=(Type=Bone,Name="clavicle_l")
    FalloffRegionScaleFactors=(PositiveWidth=1,NegativeWidth=1,PositiveHeight=-1,NegativeHeight=1)
    FalloffSize=0.3
    FlipHeightScaling=False
    FlipWidthScaling=False
    OptionalParentItem=(Type=Bone,Name="None")
    RotationOffset=(X=-70.457,Y=4.9077,Z=37.1574)
   )
   > RigVMVariableNode(Value=@n9.OutputParam) {variable=clavicleback_l}
   > n10 = &RigUnit_SphericalPoseReader(
    ActiveRegionScaleFactors=(PositiveWidth=1,NegativeWidth=1,PositiveHeight=1,NegativeHeight=1)
    ActiveRegionSize=0.15
    Debug=(bDrawDebug=false,bDraw2D=false,DebugScale=25,DebugSegments=20,DebugThickness=0.25)
    DriverAxis=(X=1,Y=0,Z=0)
    DriverItem=(Type=Bone,Name="head")
    FalloffRegionScaleFactors=(PositiveWidth=1,NegativeWidth=1,PositiveHeight=1,NegativeHeight=1)
    FalloffSize=0.32
    FlipHeightScaling=true
    FlipWidthScaling=false
    OptionalParentItem=(Type=Bone)
    RotationOffset=(X=0,Y=-30,Z=0)
   )
   > n11 = &RigUnit_SphericalPoseReader(
    ActiveRegionScaleFactors=(PositiveWidth=1,NegativeWidth=1,PositiveHeight=1,NegativeHeight=1)
    ActiveRegionSize=0.15
    Debug=(bDrawDebug=false,bDraw2D=false,DebugScale=25,DebugSegments=20,DebugThickness=0.25)
    DriverAxis=(X=1,Y=0,Z=0)
    DriverItem=(Type=Bone,Name="head")
    FalloffRegionScaleFactors=(PositiveWidth=1,NegativeWidth=1,PositiveHeight=1,NegativeHeight=1)
    FalloffSize=0.32
    FlipHeightScaling=true
    FlipWidthScaling=false
    OptionalParentItem=(Type=Bone)
    RotationOffset=(X=180,Y=30,Z=0)
   )
   n12 = RigVMVariableNode() {variable=isShovelCrop_Cape}
   RigVMFunction_ControlFlowBranch = &RigVMFunction_ControlFlowBranch(ExecuteContext=@RigVMFunction_Sequence_2.B, Condition=@n12.Value)
   n13 = RigVMVariableNode() {variable=head_up}
   &RigUnit_ModifyTransforms(
    ItemToModify=((Item=(Type=Bone,Name="head_aim_tgt"),Transform=(Rotation=(X=0,Y=0,Z=0,W=1),Translation=(X=-20,Y=-17,Z=0),Scale3D=(X=1,Y=1,Z=1))))
    WeightMinimum=0
    ExecutePin=@RigVMFunction_ControlFlowBranch.False
    Weight=@n13.Value
   )
   n14 = &RigVMFunction_AlphaInterp(
    InterpSpeedDecreasing=5
    InterpSpeedIncreasing=5
    Scale=1
    bClampResult=true
    Value=&RigVMFunction_MathDoubleMul(B=RigVMVariableNode() {variable=isCrouching}.Value, A=@n13.Value).Result
   )
   > &RigUnit_ModifyTransforms(
    ItemToModify=((Item=(Type=Bone,Name="head_aim_tgt"),Transform=(Rotation=(X=0,Y=0,Z=0,W=1),Translation=(X=0,Y=20.8825,Z=0),Scale3D=(X=1,Y=1,Z=1))))
    WeightMinimum=0
    Weight=@n14.Result
   )
   n15 = RigVMVariableNode() {variable=head_overshoot}
   > &RigUnit_ModifyTransforms(
    ItemToModify=((Item=(Type=Bone,Name="head_aim_tgt"),Transform=(Rotation=(X=0,Y=0,Z=0,W=1),Translation=(X=-10,Y=-5,Z=0),Scale3D=(X=1,Y=1,Z=1))))
    WeightMinimum=0
    Weight=@n15.Value
   )
   n16 = RigVMVariableNode() {variable=head_down}
   > &RigUnit_ModifyTransforms(
    ItemToModify=((Item=(Type=Bone,Name="head_aim_tgt"),Transform=(Rotation=(X=0,Y=0,Z=0,W=1),Translation=(X=-0.2512,Y=15.5144,Z=0.0458),Scale3D=(X=1,Y=1,Z=1))))
    WeightMinimum=0
    Weight=@n16.Value
   )
   n17 = &RigVMFunction_AlphaInterp(
    InterpSpeedDecreasing=5
    InterpSpeedIncreasing=5
    Scale=1
    bClampResult=true
    Value=&RigVMFunction_MathDoubleMul(B=RigVMVariableNode() {variable=isCrouching}.Value, A=@n16.Value).Result
   )
   > &RigUnit_ModifyTransforms(
    ItemToModify=((Item=(Type=Bone,Name="head_aim_tgt"),Transform=(Rotation=(X=0,Y=0,Z=0,W=1),Translation=(X=0,Y=-4.0726,Z=0),Scale3D=(X=1,Y=1,Z=1))))
    WeightMinimum=0
    Weight=@n17.Result
   )
   n18 = RigVMVariableNode() {variable=head_l}
   > &RigUnit_ModifyTransforms(
    ItemToModify=((Item=(Type=Bone,Name="head_aim_tgt"),Transform=(Rotation=(X=0,Y=0,Z=0,W=1),Translation=(X=-14.9042,Y=0,Z=19.7584),Scale3D=(X=1,Y=1,Z=1))))
    WeightMinimum=0
    Weight=@n18.Value
   )
   n19 = RigVMVariableNode() {variable=head_r}
   > &RigUnit_ModifyTransforms(
    ItemToModify=((Item=(Type=Bone,Name="head_aim_tgt"),Transform=(Rotation=(X=0,Y=0,Z=0,W=1),Translation=(X=-14.9042,Y=0,Z=-19.7584),Scale3D=(X=1,Y=1,Z=1))))
    WeightMinimum=0
    Weight=@n19.Value
   )
   Multiply = &RigVMFunction_MathDoubleMul(
    B=&RigVMFunction_MathBoolToFloat(Value=RigVMVariableNode() {variable=ismoving2D}.Value).Result
    A=RigVMVariableNode() {variable=isCrouching}.Value
   )
   n20 = &RigVMFunction_AlphaInterp(InterpSpeedDecreasing=5, InterpSpeedIncreasing=5, Scale=1, bClampResult=False, Value=@Multiply.Result)
   > &RigUnit_ModifyTransforms(
    ItemToModify=((Item=(Type=Bone,Name="head_aim_tgt"),Transform=(Rotation=(X=0,Y=0,Z=0,W=1),Translation=(X=0,Y=-3.0385,Z=0),Scale3D=(X=1,Y=1,Z=1))))
    WeightMinimum=0
    Weight=@n20.Result
   )
   > &RigUnit_ModifyTransforms(
    ItemToModify=((Item=(Type=Bone,Name="head_aim_tgt"),Transform=(Rotation=(X=0,Y=0,Z=0,W=1),Translation=(X=0,Y=0,Z=7.883),Scale3D=(X=1,Y=1,Z=1))))
    WeightMinimum=0
    Weight=@n10.OutputParam
   )
   > &RigUnit_ModifyTransforms(
    ItemToModify=((Item=(Type=Bone,Name="head_aim_tgt"),Transform=(Rotation=(X=0,Y=0,Z=0,W=1),Translation=(X=0,Y=0,Z=-7.883),Scale3D=(X=1,Y=1,Z=1))))
    WeightMinimum=0
    Weight=@n11.OutputParam
   )
   RigVMFunction_MathBoolToFloat = &RigVMFunction_MathBoolToFloat(Value=RigVMVariableNode() {variable=isShovelCrop_Cape}.Value)
   > &RigUnit_ModifyTransforms(
    ItemToModify=((Item=(Type=Bone,Name="head_aim_tgt"),Transform=(Rotation=(X=0,Y=0,Z=0,W=1),Translation=(X=0,Y=-5.0785,Z=0),Scale3D=(X=1,Y=1,Z=1))))
    WeightMinimum=0
    Weight=@RigVMFunction_MathBoolToFloat.Result
   )
   n21 = &RigVMFunction_AlphaInterp(
    InterpSpeedDecreasing=5
    InterpSpeedIncreasing=5
    Scale=1
    bClampResult=true
    Value=&RigVMFunction_MathDoubleMul(A=@RigVMFunction_MathBoolToFloat.Result, B=RigVMVariableNode() {variable=isCrouching}.Value).Result
   )
   > &RigUnit_ModifyTransforms(
    ItemToModify=((Item=(Type=Bone,Name="head_aim_tgt"),Transform=(Rotation=(X=0,Y=0,Z=0,W=1),Translation=(X=0,Y=-0.4987,Z=0),Scale3D=(X=1,Y=1,Z=1))))
    WeightMinimum=0
    Weight=@n21.Result
   )
   &RigUnit_ModifyTransforms(
    ItemToModify=((Item=(Type=Bone,Name="base_head_aim_tgt"),Transform=(Rotation=(X=0,Y=0,Z=0,W=1),Translation=(X=-68.7118,Y=-69.055,Z=30.3818),Scale3D=(X=1,Y=1,Z=1))))
    WeightMinimum=0
    ExecutePin=@RigVMFunction_ControlFlowBranch.True
    Weight=@n13.Value
   )
   > &RigUnit_ModifyTransforms(
    ItemToModify=((Item=(Type=Bone,Name="head_aim_tgt"),Transform=(Rotation=(X=0,Y=0,Z=0,W=1),Translation=(X=0,Y=0,Z=0),Scale3D=(X=1,Y=1,Z=1))))
    WeightMinimum=0
    Weight=@n15.Value
   )
   > &RigUnit_ModifyTransforms(
    ItemToModify="((Item=(Type=Bone,Name=\"head_aim_tgt\"),Transform=(Rotation=(X=0,Y=0,Z=0,W=1),Translation=(X=-12.4947,Y=-49.6648,Z=0),Scale3D=(X=1,Y=1,Z=1))),(Item=(Type=Bone,Name=\"offset_hair_back_01\"),Transform=(Rotation=(X=0,Y=0,Z=-0.342,W=0.9397),Translation=(X=0,Y=0,Z ...(truncated)"
    WeightMinimum=0
    Weight=@n16.Value
   )
   > &RigUnit_ModifyTransforms(
    ItemToModify=((Item=(Type=Bone,Name="head_aim_tgt"),Transform=(Rotation=(X=0,Y=0,Z=0,W=1),Translation=(X=0,Y=-14.7754,Z=0),Scale3D=(X=1,Y=1,Z=1))))
    WeightMinimum=0
    Weight=@n18.Value
   )
   > &RigUnit_ModifyTransforms(
    ItemToModify=((Item=(Type=Bone,Name="head_aim_tgt"),Transform=(Rotation=(X=0,Y=0,Z=0,W=1),Translation=(X=0,Y=5.9415,Z=0),Scale3D=(X=1,Y=1,Z=1))))
    WeightMinimum=0
    Weight=@n19.Value
   )
   &RigUnit_ModifyTransforms(
    ItemToModify=((Item=(Type=Bone,Name="offset_hair_back_01"),Transform=(Rotation=(X=-0.0773,Y=0.4852,Z=-0.172,W=0.8538),Translation=(X=0,Y=4,Z=0),Scale3D=(X=1,Y=1,Z=1))))
    WeightMinimum=0
    ExecutePin=@RigVMFunction_Sequence_2.E
    Weight=RigVMVariableNode() {variable=clavicleback_l}.Value
   )
   > &RigUnit_ModifyTransforms(
    ItemToModify=((Item=(Type=Bone,Name="offset_hair_back_01"),Transform=(Rotation=(X=-0.0076,Y=-0.1735,Z=-0.043,W=0.9839),Translation=(X=0,Y=4,Z=0),Scale3D=(X=1,Y=1,Z=1))))
    WeightMinimum=0
    Weight=RigVMVariableNode() {variable=clavicleback_r}.Value
   )
   AlphaInterp = &RigVMFunction_AlphaInterp(
    InterpSpeedDecreasing=10
    InterpSpeedIncreasing=10
    Value=0
    bClampResult=true
    Scale=&RigVMFunction_MathBoolToFloat(Value=RigVMVariableNode() {variable=ismoving2D}.Value).Result
   )
   > &RigUnit_ModifyTransforms(
    ItemToModify=((Item=(Type=Bone,Name="offset_hair_back_01"),Transform=(Rotation=(X=0,Y=0,Z=-0.1736,W=0.9848),Translation=(X=0,Y=0,Z=0),Scale3D=(X=1,Y=1,Z=1))))
    WeightMinimum=0
    Weight=@AlphaInterp.Result
   )
   &RigUnit_ModifyTransforms(
    ItemToModify=((Item=(Type=Bone,Name="None"),Transform=(Translation=(X=0,Y=0,Z=0),Scale3D=(X=1,Y=1,Z=1))))
    Weight=1
    WeightMinimum=0
    ExecutePin=@RigVMFunction_Sequence_2.F
   )
   n22 = &RigVMFunction_MathBoolToFloat(Value=RigVMVariableNode() {variable=isShovelCrop_Cape}.Value)
   n23 = &RigVMFunction_MathDoubleMul(B=@n22.Result, A=RigVMVariableNode() {variable=head_l}.Value)
   > &RigUnit_ModifyTransforms(
    ItemToModify=((Item=(Type=Bone,Name="offset_hair_back_01"),Transform=(Rotation=(X=0,Y=0,Z=0.1305,W=0.9914),Translation=(X=0,Y=0,Z=0),Scale3D=(X=1,Y=1,Z=1))))
    WeightMinimum=0
    Weight=@n23.Result
   )
   n24 = RigVMVariableNode() {variable=rigidbody_enable}
   > &RigUnit_ModifyTransforms(
    ItemToModify=((Item=(Type=Bone,Name="offset_hair_back_01"),Transform=(Rotation=(X=-0.0189,Y=-0.2156,Z=-0.0851,W=0.9726),Translation=(X=0,Y=0,Z=0),Scale3D=(X=1,Y=1,Z=1))))
    WeightMinimum=0
    Weight=&RigVMFunction_MathDoubleMul(A=@n23.Result, B=&RigVMFunction_MathBoolToFloat(Value=@n24.Value).Result).Result
   )
   > &RigUnit_ModifyTransforms(
    ItemToModify=((Item=(Type=Bone,Name="offset_hair_back_01"),Transform=(Rotation=(X=0,Y=0,Z=0.1305,W=0.9914),Translation=(X=0,Y=0,Z=0),Scale3D=(X=1,Y=1,Z=1))))
    WeightMinimum=0
    Weight=&RigVMFunction_MathDoubleMul(B=@n22.Result, A=RigVMVariableNode() {variable=head_r}.Value).Result
   )
   > &RigUnit_ModifyTransforms(ItemToModify=((Item=(Type=Bone,Name="None"),Transform=(Translation=(X=0,Y=0,Z=0),Scale3D=(X=1,Y=1,Z=1)))), Weight=1, WeightMinimum=0)
   > &RigUnit_ModifyTransforms(ItemToModify=((Item=(Type=Bone,Name="None"),Transform=(Translation=(X=0,Y=0,Z=0),Scale3D=(X=1,Y=1,Z=1)))), Weight=1, WeightMinimum=0)
   > &RigUnit_ModifyTransforms(ItemToModify=((Item=(Type=Bone,Name="None"),Transform=(Translation=(X=0,Y=0,Z=0),Scale3D=(X=1,Y=1,Z=1)))), Weight=1, WeightMinimum=0)
   &RigUnit_AimConstraintLocalSpaceOffset(
    Child=(Type=Bone,Name="aim_hair_back_01")
    Filter=(bX=true,bY=true,bZ=true)
    Parents=((Item=(Type=Bone,Name="head_aim_tgt"),Weight=1))
    UpAxis=(X=0,Y=-1,Z=0)
    Weight=1
    WorldUp=(Target=(X=0,Y=-1,Z=0),Kind=Location,Space=(Type=Bone,Name="head_aim_tgt"))
    bMaintainOffset=true
    ExecutePin=@RigVMFunction_Sequence_2.D
   )
   RigVMFunctionReturnNode(Result=(), ExecuteContext=@n2.B)
   RigVMFunction_MathBoolAnd = &RigVMFunction_MathBoolAnd(A=@n12.Value, B=RigVMVariableNode() {variable=rigidbody_enable}.Value)
  ]}
 RigVMFunctionLibrary {links [], nodes []}
 RigVMFunction_Sequence_3_ContainedGraph {nodes [
   n1 = RigVMFunctionEntryNode()
   n2 = &RigVMFunction_Sequence(ExecuteContext=@n1.ExecuteContext)
   n3 = &RigVMFunction_Sequence(ExecuteContext=@n2.B)
   n4 = &RigVMFunction_Sequence(ExecuteContext=@n3.B)
   n5 = &RigVMFunction_Sequence(ExecuteContext=@n4.B)
   RigVMFunctionReturnNode(B=@n3.A, A=@n2.A, E=@n5.A, F=@n5.B, D=@n4.A)
  ]}
 RigVMModel {nodes [
   n1 = &RigUnit_PrepareForExecution()
   addFKCtrl = RigVMFunctionReferenceNode(
    Custom_FKCtrl=()
    Options_FKCtrl="(pelvis_1_#1=False,spine_3_#2=False,clavicle_5_#3=False,arm_7_#4=False,leg_9_#5=False,head_11_46DC15EF465C ...(truncated)"
    Shape_Settings="(pelvis_2_#6=(bVisible=True,Name=\"Hexagon_Thin\",Color=(R=0.0117,G=0.8281,B=1,A=1),Transform=(Rotation=(X=0,Y=-0.7071,Z=0,W=0.7071),Translation=(X=2,Y=0,Z=0),Scale3D=(X=5,Y=5,Z=7))),spine_4_#7=(bVi ...(truncated)"
    ExecuteContext=@n1.ExecutePin
   )
   RigUnit_BeginExecution = &RigUnit_BeginExecution()
   setFKCtrl = RigVMFunctionReferenceNode(setFK_Ctrl=false, ExecuteContext=@RigUnit_BeginExecution.ExecutePin)
   > HairSetup = RigVMCollapseNode()
  ]}
]
hierarchy [
 root {loc=<0 0 0>, rot=<0 0 0 1>}
 pelvis {parent=root, loc=<0 -0.5301 98.0571>, rot=<0.0127 0.707 -0.0127 -0.707>}
 spine_01 {parent=pelvis, loc=<6.1208 -0.9328 0>, rot=<0 0 0.034 -0.9994>}
 spine_02 {parent=spine_01, loc=<6.5997 -0.2095 0>, rot=<0 0 0.0238 -0.9997>}
 spine_03 {parent=spine_02, loc=<6.4914 -0.1237 0>, rot=<0 0 0.0149 -0.9999>}
 spine_04 {parent=spine_03, loc=<6.9598 0.0794 0>, rot=<0 0 -0.0132 -0.9999>}
 spine_05 {parent=spine_04, loc=<6.9987 0 0>, rot=<0 0 -0.0015 -1>}
 neck_01 {parent=spine_05, loc=<11.3138 0.1529 0>, rot=<0 0 -0.2003 -0.9797>}
 neck_02 {parent=neck_01, loc=<4.5627 0 0>, rot=<0 0 0.0504 -0.9987>}
 head {parent=neck_02, loc=<5.3914 0 0>, rot=<0 0 0.115 -0.9934>}
 base_hair_MSTR {parent=head, loc=<8.645 0.062 0>, rot=<0.0029 -0.7071 -0.0029 -0.7071>}
 base_hair_front_00_L {parent=base_hair_MSTR, loc=<6.0971 2.7421 -0.4213>, rot=<0.7134 -0.5072 -0.2715 -0.4001>}
 dyn_hair_front_00_L {parent=base_hair_front_00_L, loc=<0 0 0>, rot=<0 0 0 1>}
 dyn_hair_front_right_00_L {parent=dyn_hair_front_00_L, loc=<4.807 0 2.0072>, rot=<0.0312 0.2514 -0.1193 -0.96>}
 dyn_hair_front_right_01_L {parent=dyn_hair_front_right_00_L, loc=<4.1406 0 0>, rot=<0.0119 0.2169 -0.0537 -0.9746>}
 dyn_hair_front_right_02_L {parent=dyn_hair_front_right_01_L, loc=<4.824 0 0>, rot=<0 0 0 1>}
 dyn_hair_front_left_00_L {parent=dyn_hair_front_00_L, loc=<6.2488 -0.3848 -0.9001>, rot=<0.1954 -0.3464 0.13 -0.9083>}
 dyn_hair_front_left_01_L {parent=dyn_hair_front_left_00_L, loc=<4.8567 0 0>, rot=<0.0187 -0.2418 0.0071 -0.9701>}
 base_hair_front_00_R {parent=base_hair_MSTR, loc=<-5.5645 3.8041 -0.4146>, rot=<0.4874 -0.727 -0.3957 -0.278>}
 dyn_hair_front_00_R {parent=base_hair_front_00_R, loc=<0 0 0>, rot=<0 0 0 1>}
 dyn_hair_front_right_00_R {parent=dyn_hair_front_00_R, loc=<5.5082 0 0.76>, rot=<-0.0364 0.3086 0.1866 -0.932>}
 dyn_hair_front_right_01_R {parent=dyn_hair_front_right_00_R, loc=<5.1464 0 0>, rot=<0.0021 0.1906 -0.0083 -0.9816>}
 dyn_hair_front_right_02_R {parent=dyn_hair_front_right_01_R, loc=<4.824 0 0>, rot=<0 0 0 1>}
 dyn_hair_front_left_00_R {parent=dyn_hair_front_00_R, loc=<6.2488 -0.3848 -0.9001>, rot=<0.8701 0.3193 -0.2167 0.3067>}
 dyn_hair_front_left_01_R {parent=dyn_hair_front_left_00_R, loc=<4.8567 0 0>, rot=<0.0102 0.0424 0.2332 -0.9714>}
 base_hair_back_00 {parent=base_hair_MSTR, loc=<0 -10.0778 4.0736>, rot=<0.593 -0.3851 0.593 -0.3851>}
 dyn_hair_back_00 {parent=base_hair_back_00, loc=<0 0 0>, rot=<0 0 0 1>}
 aim_hair_back_01 {parent=dyn_hair_back_00, loc=<5.0686 0 0>, rot=<0 0 -0.6361 -0.7716>}
 offset_hair_back_01 {parent=aim_hair_back_01, loc=<0 0 0>, rot=<0 0 0 1>}
 dyn_hair_back_01 {parent=offset_hair_back_01, loc=<0 0 0>, rot=<0 0 0 1>}
 dyn_hair_back_02 {parent=dyn_hair_back_01, loc=<13.4109 2.3657 0>, rot=<-0.0042 -0.0007 -0.2041 -0.9789>}
 dyn_hair_back_03 {parent=dyn_hair_back_02, loc=<18.1597 0.7639 0>, rot=<0 0 -0.0901 -0.9959>}
 dyn_hair_back_04 {parent=dyn_hair_back_03, loc=<12.7074 -0.8753 0.0575>, rot=<-0.0003 -0.0134 -0.0179 -0.9997>}
 dyn_hair_back_05 {parent=dyn_hair_back_04, loc=<18.8141 0 0>, rot=<-0.0003 0.0002 -0.0905 -0.9959>}
 base_hair_back_00_a {parent=dyn_hair_back_04, loc=<-3.2867 -0.201 -9.3915>, rot=<-0.4665 -0.0534 0.0035 -0.8829>}
 dyn_hair_back_00_a {parent=base_hair_back_00_a, loc=<0 0 0>, rot=<0 0 0 1>}
 dyn_hair_back_01_a {parent=dyn_hair_back_00_a, loc=<10.1605 0 0>, rot=<0.1288 0.0877 -0.0344 -0.9872>}
 dyn_hair_back_02_a {parent=dyn_hair_back_01_a, loc=<9.8702 0 0>, rot=<-0.1027 0.1704 -0.0815 -0.9766>}
 dyn_hair_back_03_a {parent=dyn_hair_back_02_a, loc=<8.3959 -0.6088 0>, rot=<-0.0081 0.2165 0.0364 -0.9756>}
 base_hair_back_00_c {parent=dyn_hair_back_04, loc=<8.7187 5.4715 8.9238>, rot=<-0.3544 -0.0179 0.0471 -0.9337>}
 dyn_hair_back_00_c {parent=base_hair_back_00_c, loc=<0 0 0>, rot=<0 0 0.0034 -1>}
 dyn_hair_back_01_c {parent=dyn_hair_back_00_c, loc=<6.0182 0 0>, rot=<0 0 0.1813 -0.9834>}
 dyn_hair_back_02_c {parent=dyn_hair_back_01_c, loc=<5.211 0 0>, rot=<0 0 0.0458 -0.999>}
 base_hair_back_00_e {parent=dyn_hair_back_04, loc=<5.0193 -4.4458 11.0826>, rot=<-0.1865 -0.0092 -0.0142 -0.9823>}
 dyn_hair_back_00_e {parent=base_hair_back_00_e, loc=<0 0 0>, rot=<0 0 0 1>}
 dyn_hair_back_01_e {parent=dyn_hair_back_00_e, loc=<6.5965 0 0>, rot=<0 -0.1013 0 -0.9949>}
 dyn_hair_back_02_e {parent=dyn_hair_back_01_e, loc=<5.6787 0 0>, rot=<0 -0.1444 0 -0.9895>}
 dyn_hair_back_03_e {parent=dyn_hair_back_02_e, loc=<4.5988 0 0>, rot=<0 -0.0913 0 -0.9958>}
 base_hair_back_00_d {parent=dyn_hair_back_03, loc=<4.2807 2.5303 -7.3187>, rot=<-0.3349 -0.098 -0.0482 -0.9359>}
 dyn_hair_back_00_d {parent=base_hair_back_00_d, loc=<0 0 0>, rot=<0 0 0 1>}
 dyn_hair_back_01_d {parent=dyn_hair_back_00_d, loc=<11.3314 0 0>, rot=<-0.0014 0.0776 0.018 -0.9968>}
 dyn_hair_back_02_d {parent=dyn_hair_back_01_d, loc=<7.2032 0 0>, rot=<0 0.0896 0 -0.996>}
 dyn_hair_back_03_d {parent=dyn_hair_back_02_d, loc=<7.4443 0 0>, rot=<0 0.122 0 -0.9925>}
 base_ponytail_ribbon_MSTR {parent=base_hair_MSTR, loc=<-0.0149 -7.2427 5.7351>, rot=<0 0 0 1>}
 base_ribbon_side_00_lf {parent=base_ponytail_ribbon_MSTR, loc=<1.7127 -0.1381 -0.14>, rot=<0.6688 -0.3841 -0.1806 -0.6104>}
 dyn_ribbon_side_00_lf {parent=base_ribbon_side_00_lf, loc=<0 0 0>, rot=<0 0 0 1>}
 dyn_ribbon_side_01_lf {parent=dyn_ribbon_side_00_lf, loc=<4.3643 0 0>, rot=<0 0 -0.2176 -0.976>}
 base_lace_side_00_lf {parent=base_ponytail_ribbon_MSTR, loc=<3.2417 -0.6958 -1.4784>, rot=<0.5921 -0.5369 -0.2284 -0.5558>}
 dyn_lace_side_00_lf {parent=base_lace_side_00_lf, loc=<0 0 0>, rot=<0 0 0 1>}
 dyn_lace_side_01_lf {parent=dyn_lace_side_00_lf, loc=<3.7098 0 0>, rot=<0.0349 -0.1119 -0.2499 -0.9611>}
 dyn_lace_side_02_lf {parent=dyn_lace_side_01_lf, loc=<3.0262 0 0>, rot=<0 0 0 1>}
 base_ribbon_side_00_rt {parent=base_ponytail_ribbon_MSTR, loc=<-1.8343 -0.1266 -0.262>, rot=<-0.2504 -0.4776 0.7648 -0.3525>}
 dyn_ribbon_side_00_rt {parent=base_ribbon_side_00_rt, loc=<0 0 0>, rot=<0 0 0 1>}
 dyn_ribbon_side_01_rt {parent=dyn_ribbon_side_00_rt, loc=<4.3643 0 0>, rot=<0 0 -0.2176 -0.976>}
 base_lace_side_00_rt {parent=base_ponytail_ribbon_MSTR, loc=<-3.0043 -0.7351 -1.3581>, rot=<-0.3456 -0.4064 0.7188 -0.4457>}
 dyn_lace_side_00_rt {parent=base_lace_side_00_rt, loc=<0 0 0>, rot=<0 0 0 1>}
 dyn_lace_side_01_rt {parent=dyn_lace_side_00_rt, loc=<3.7098 0 0>, rot=<0.0121 -0.0242 -0.252 -0.9673>}
 dyn_lace_side_02_rt {parent=dyn_lace_side_01_rt, loc=<3.0262 0 0>, rot=<0 0 0 1>}
 base_head_aim_tgt {parent=spine_05, loc=<12.9246 -19.8279 0.2751>, rot=<0.0075 -0.0008 0.0004 -1>}
 head_aim_tgt {parent=base_head_aim_tgt, loc=<0 0 0>, rot=<0 0 0 1>}
 clavicle_l {parent=spine_05, loc=<5.4831 -0.1053 -3.2946>, rot=<0.0465 -0.7289 -0.009 -0.6829>}
 upperarm_l {parent=clavicle_l, loc=<9.8658 -1.2136 -0.4105>, rot=<0.0164 -0.3245 -0.0269 -0.9454>}
 lowerarm_l {parent=upperarm_l, loc=<26.465 -0.8345 0.259>, rot=<0.0036 0.0161 -0.2298 -0.9731>}
 clavicle_r {parent=spine_05, loc=<5.4831 -0.1053 3.2946>, rot=<0.7289 0.0465 -0.6829 0.009>}
 upperarm_r {parent=clavicle_r, loc=<-9.8658 1.2136 0.4105>, rot=<0.0164 -0.3245 -0.0269 -0.9454>}
 lowerarm_r {parent=upperarm_r, loc=<-26.465 0.8345 -0.259>, rot=<0.0036 0.0161 -0.2298 -0.9731>}
]
variables [
 clavicleback_l {type=double, default=0}
 clavicleback_r {type=double, default=0}
 head_down {type=double, default=0}
 head_l {type=double, default=0}
 head_overshoot {type=double, default=0}
 head_r {type=double, default=0}
 head_up {type=double, default=0}
 isCrouching {type=double, default=0}
 isShovelCrop_Cape {type=bool, default=False}
 isWearingBackBlingAndCape {type=bool, default=False}
 ismoving2D {type=bool, default=False}
 iswearingBackbling {type=double, default=0}
 rigidbody_enable {type=bool, default=False}
]
