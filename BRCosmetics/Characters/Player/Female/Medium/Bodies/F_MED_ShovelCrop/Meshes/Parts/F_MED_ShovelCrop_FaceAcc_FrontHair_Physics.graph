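# Physics asset for the front hair part of the female medium ShovelCrop character, defining capsule bodies and joint limits for hair motion.
# Mirrored L and R trees hang from base bones, each branching at the front root into a two-link left strand and a three-link right strand.
# Constraints lock twist and keep swing narrow, widest at the right-strand tips.
# Two unconstrained head capsules provide skull collision shapes.
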PhysicsAsset /BRCosmetics/Characters/Player/Female/Medium/Bodies/F_MED_ShovelCrop/Meshes/Parts/F_MED_ShovelCrop_FaceAcc_FrontHair_Physics
bodies {
 base_hair_MSTR [{shapeType=Capsule, length=1, radius=1, shapeName=base_hair_MSTR_capsule}]
 base_hair_front_00_L [{shapeType=Capsule, length=1, radius=1, shapeName=base_hair_front_00_L_capsule, rotation=<rot 0 0 -180>}]
 base_hair_front_00_R [{shapeType=Capsule, length=1, radius=1, shapeName=base_hair_front_00_R_capsule}]
 dyn_hair_front_00_L [{shapeType=Capsule, length=5, radius=2.505, shapeName=dyn_hair_front_00_L_capsule, center=<2 0 0>}]
 dyn_hair_front_00_R [{shapeType=Capsule, length=5, radius=2.505, shapeName=dyn_hair_front_00_R_capsule, center=<2 0 0>, rotation=<rot 0 0 -180>}]
 dyn_hair_front_left_00_L [{shapeType=Capsule, length=2.6121, radius=1.8644, shapeName=dyn_hair_front_left_00_L_capsule, center=<2 0 0>}]
 dyn_hair_front_left_00_R [{shapeType=Capsule, length=2.6121, radius=1.8644, shapeName=dyn_hair_front_left_00_R_capsule, center=<2 0 0>, rotation=<rot 0 0 -180>}]
 dyn_hair_front_left_01_L [{shapeType=Capsule, length=3.9954, radius=2.2527, shapeName=dyn_hair_front_left_01_L_capsule, center=<2 0 0>}]
 dyn_hair_front_left_01_R [{shapeType=Capsule, length=3.9954, radius=2.2527, shapeName=dyn_hair_front_left_01_R_capsule, center=<2 0 0>, rotation=<rot 0 0 -180>}]
 dyn_hair_front_right_00_L [{shapeType=Capsule, length=2.7083, radius=1.9087, shapeName=dyn_hair_front_right_00_L_capsule, center=<2 0 0>}]
 dyn_hair_front_right_00_R [{shapeType=Capsule, length=2.7083, radius=1.9087, shapeName=dyn_hair_front_right_00_R_capsule, center=<2 0 0>, rotation=<rot 0 0 -180>}]
 dyn_hair_front_right_01_L [{shapeType=Capsule, length=7, radius=2, shapeName=dyn_hair_front_right_01_L_capsule, center=<2 0 0>}]
 dyn_hair_front_right_01_R [{shapeType=Capsule, length=7, radius=2, shapeName=dyn_hair_front_right_01_R_capsule, center=<2 1.1 0>, rotation=<rot 0 0 -180>}]
 dyn_hair_front_right_02_L [{
   shapeType=Capsule
   length=4.1861
   radius=2.614
   shapeName=dyn_hair_front_right_02_L_capsule
   center=<1.5 -1.3 0>
   rotation=<rot -4.6976 1.7139 -20.0703>
  }]
 dyn_hair_front_right_02_R [{
   shapeType=Capsule
   length=4.1861
   radius=2.114
   shapeName=dyn_hair_front_right_02_R_capsule
   center=<1.5 0 0>
   rotation=<rot -37.1585 16.0129 154.5863>
  }]
 head [
  {shapeType=Capsule, length=4.7226, radius=5.8542, shapeName=head_capsule, center=<0 2.3 0>, rotation=<rot 90 0 0>}
  {shapeType=Capsule, length=3.5, radius=4.5, shapeName=head_capsule, center=<1.845 -1.9194 0>}
 ]
}
constraints [
 {bone1Name=dyn_hair_front_00_L, bone2Name=base_hair_front_00_L, swing1LimitDegrees=5, swing2LimitDegrees=5, twistLimitDegrees=0}
 {bone1Name=dyn_hair_front_left_00_L, bone2Name=dyn_hair_front_00_L, swing1LimitDegrees=15, swing2LimitDegrees=5, twistLimitDegrees=0}
 {bone1Name=dyn_hair_front_left_01_L, bone2Name=dyn_hair_front_left_00_L, swing1LimitDegrees=15, swing2LimitDegrees=5, twistLimitDegrees=0}
 {bone1Name=dyn_hair_front_right_00_L, bone2Name=dyn_hair_front_00_L, swing1LimitDegrees=5, swing2LimitDegrees=5, twistLimitDegrees=0}
 {bone1Name=dyn_hair_front_right_01_L, bone2Name=dyn_hair_front_right_00_L, swing1LimitDegrees=10, swing2LimitDegrees=10, twistLimitDegrees=0}
 {bone1Name=dyn_hair_front_right_02_L, bone2Name=dyn_hair_front_right_01_L, swing1LimitDegrees=40, swing2LimitDegrees=5, twistLimitDegrees=0}
 {bone1Name=dyn_hair_front_00_R, bone2Name=base_hair_front_00_R, swing1LimitDegrees=5, swing2LimitDegrees=5, twistLimitDegrees=0}
 {bone1Name=dyn_hair_front_left_00_R, bone2Name=dyn_hair_front_00_R, swing1LimitDegrees=15, swing2LimitDegrees=5, twistLimitDegrees=0}
 {bone1Name=dyn_hair_front_left_01_R, bone2Name=dyn_hair_front_left_00_R, swing1LimitDegrees=15, swing2LimitDegrees=5, twistLimitDegrees=0}
 {bone1Name=dyn_hair_front_right_00_R, bone2Name=dyn_hair_front_00_R, swing1LimitDegrees=5, swing2LimitDegrees=5, twistLimitDegrees=0}
 {bone1Name=dyn_hair_front_right_01_R, bone2Name=dyn_hair_front_right_00_R, swing1LimitDegrees=10, swing2LimitDegrees=10, twistLimitDegrees=0}
 {bone1Name=dyn_hair_front_right_02_R, bone2Name=dyn_hair_front_right_01_R, swing1LimitDegrees=35, swing2LimitDegrees=5, twistLimitDegrees=0}
]
