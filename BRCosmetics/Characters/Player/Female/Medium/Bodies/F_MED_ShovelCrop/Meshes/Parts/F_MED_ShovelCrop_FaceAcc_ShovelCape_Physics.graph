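# Physics asset for the ShovelCape face accessory on the female medium ShovelCrop character, rigged on a chain of back-hair bones.
# Base and first dynamic capsules are locked to the head, while segments 01–05 form a separate chain from offset_hair_back_01 with swing limits and zero-degree twist limits.
# Capsules on head, neck, spine, pelvis and upper arms supply body shapes, with tiny capsules on the aim and offset bones.
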PhysicsAsset /BRCosmetics/Characters/Player/Female/Medium/Bodies/F_MED_ShovelCrop/Meshes/Parts/F_MED_ShovelCrop_FaceAcc_ShovelCape_Physics
bodies {
 aim_hair_back_01 [{shapeType=Capsule, length=0.505, radius=0.505, shapeName=aim_hair_back_01_capsule, rotation=<rot 0 -90 0>}]
 base_hair_back_00 [{shapeType=Capsule, length=8, radius=6.5, shapeName=base_hair_back_00_capsule, center=<2 0 0>, rotation=<rot 90 0 0>}]
 dyn_hair_back_00 [{shapeType=Capsule, length=8, radius=6.5, shapeName=dyn_hair_back_00_capsule, center=<2 0 0>, rotation=<rot 90 0 0>}]
 dyn_hair_back_01 [{shapeType=Capsule, length=8, radius=6.5, shapeName=dyn_hair_back_01_capsule, center=<2 0 0>, rotation=<rot 90 0.0001 0>}]
 dyn_hair_back_02 [{shapeType=Capsule, length=8, radius=6.5, shapeName=dyn_hair_back_02_capsule, center=<2 0 0>, rotation=<rot 90 0 0>}]
 dyn_hair_back_03 [{shapeType=Capsule, length=7, radius=7.5, shapeName=dyn_hair_back_03_capsule, center=<2 0 0>, rotation=<rot 90 0 0>}]
 dyn_hair_back_04 [{shapeType=Capsule, length=13.5, radius=9, shapeName=dyn_hair_back_04_capsule, center=<1 0 0>, rotation=<rot 0 -180 180>}]
 dyn_hair_back_05 [{shapeType=Capsule, length=7, radius=7.5, shapeName=dyn_hair_back_05_capsule, center=<1 0 0>, rotation=<rot 90 0 0>}]
 head [{shapeType=Capsule, length=16.5919, radius=8.4044, center=<10.1042 -8.1456 -0.1903>, rotation=<rot 0 19.329 -89.5363>}]
 neck_01 [{shapeType=Capsule, length=1, radius=9, shapeName=neck_01_capsule, center=<-2.3327 -3.4964 0>}]
 neck_02 [{shapeType=Capsule, length=1, radius=9, shapeName=neck_02_capsule}]
 offset_hair_back_01 [{shapeType=Capsule, length=0.505, radius=0.505, shapeName=offset_hair_back_01_capsule, rotation=<rot 0 -90 0>}]
 pelvis [{shapeType=Capsule, length=11.5, radius=15, shapeName=pelvis_capsule, center=<-1.9187 -0.8442 0>}]
 spine_01 [{shapeType=Capsule, length=7.5, radius=12.5, shapeName=spine_01_capsule}]
 spine_02 [{shapeType=Capsule, length=1.5, radius=12.5, shapeName=spine_02_capsule}]
 spine_03 [{shapeType=Capsule, length=5.5, radius=14, shapeName=spine_03_capsule, center=<0 -1.5255 0>}]
 spine_04 [{shapeType=Capsule, length=7, radius=12.5, shapeName=spine_04_capsule}]
 spine_05 [{shapeType=Capsule, length=8.5, radius=12.5, shapeName=spine_05_capsule, center=<0 -2.2208 0>}]
 upperarm_l [{shapeType=Capsule, length=24.5, radius=6, shapeName=upperarm_l_capsule, center=<10.3061 -0.6291 0>, rotation=<rot 90 0 0>}]
 upperarm_r [{shapeType=Capsule, length=24.5, radius=6, shapeName=upperarm_r_capsule, center=<-10.3061 0.6291 0>, rotation=<rot 90 180 0>}]
}
constraints [
 {
  bone1Name=base_hair_back_00
  bone2Name=head
  swing1LimitDegrees=45
  swing1Motion=Locked
  swing2LimitDegrees=45
  swing2Motion=Locked
  twistLimitDegrees=45
  twistMotion=Locked
 }
 {
  bone1Name=dyn_hair_back_00
  bone2Name=base_hair_back_00
  swing1LimitDegrees=5
  swing1Motion=Locked
  swing2LimitDegrees=0
  swing2Motion=Locked
  twistLimitDegrees=0
  twistMotion=Locked
 }
 {bone1Name=dyn_hair_back_01, bone2Name=offset_hair_back_01, swing1LimitDegrees=25, swing2LimitDegrees=15, twistLimitDegrees=0}
 {bone1Name=dyn_hair_back_02, bone2Name=dyn_hair_back_01, swing1LimitDegrees=25, swing2LimitDegrees=15, twistLimitDegrees=0}
 {bone1Name=dyn_hair_back_03, bone2Name=dyn_hair_back_02, swing1LimitDegrees=25, swing2LimitDegrees=5, twistLimitDegrees=0}
 {bone1Name=dyn_hair_back_04, bone2Name=dyn_hair_back_03, swing1LimitDegrees=20, swing2LimitDegrees=10, twistLimitDegrees=0}
 {bone1Name=dyn_hair_back_05, bone2Name=dyn_hair_back_04, swing1LimitDegrees=30, swing2LimitDegrees=5, twistLimitDegrees=0}
]
