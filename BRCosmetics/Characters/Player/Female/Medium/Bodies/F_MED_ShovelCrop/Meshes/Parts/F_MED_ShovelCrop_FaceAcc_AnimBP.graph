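# Post-process AnimBP for the F_MED_ShovelCrop face accessory, animating hair, ribbons and laces.
# Copies the parent pose through a Control Rig; without rigid bodies, a Trail controller drives the back hair.
# The physics path chains AnimDynamics ribbons and laces into RigidBody passes, with cape-specific front-hair physics.
# Back hair then uses AnimDynamics with spherical body limits, or head and spine constraints when a backpack is worn.
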AnimBlueprint /BRCosmetics/Characters/Player/Female/Medium/Bodies/F_MED_ShovelCrop/Meshes/Parts/F_MED_ShovelCrop_FaceAcc_AnimBP
target_skeleton=&/F_MED_ShovelCrop_FaceAcc_Skeleton
graphs [
 AnimGraph {nodes [
   Output Pose(Result=Blend Poses by bool(BlendPose_0=Use cached pose 'Rigidbody_Dynamics_2'(), BlendPose_1=Use cached pose 'trailCache'(), bActiveValue=Get bRigidBodyEnabled()) {blendType=Linear}) {layerGroup=None}
   n1 = Control Rig!(
    Source=Copy Pose From Mesh() {bUseAttachedParent=true}
    Alpha
    isCrouching:double=Get isCrouching() -> (isCrouching:double)
    ismoving2D:bool=Get bIsMoving2D()
    iswearingBackbling:double=Get iswearingbackbling() -> (iswearingbackbling:double)
    isShovelCrop_Cape:bool=Get isShovelCrop_Cape() -> (isShovelCrop_Cape:bool)
    isWearingBackBlingAndCape:bool=Get Is Wearing Backpack() -> (Is Wearing Backpack:bool)
    rigidbody_enable:bool=Get bRigidBodyEnabled()
   ) -> (Pose) {controlRigAssetReference {blueprintRigClass=&/F_MED_ShovelCrop_FaceAcc_CtrlRig.F_MED_ShovelCrop_FaceAcc_CtrlRig_C}}
   Save cached pose 'ControlRig_Setup'(Pose=@n1) {cacheName=ControlRig_Setup}
   n2 = "Transform (Modify) Bone - Bone: dyn_hair_back_01"!(ComponentPose=Local To Component(LocalPose=Use cached pose 'ControlRig_Setup'()), Alpha=1, Rotation="0,0,0") -> (Pose) {rotationMode=BMM_Additive, rotationSpace=BCS_BoneSpace, alphaBoolBlend {blendOption=Linear}, boneToModify {boneName=dyn_hair_back_01}}
   n3 = "Transform (Modify) Bone - Bone: dyn_hair_back_02"!(ComponentPose=@n2, Alpha=1, Rotation="0,0,0") -> (Pose) {rotationMode=BMM_Additive, rotationSpace=BCS_BoneSpace, alphaBoolBlend {blendOption=Linear}, boneToModify {boneName=dyn_hair_back_02}}
   n4 = "Transform (Modify) Bone - Bone: dyn_hair_back_01"!(ComponentPose=@n3, bAlphaBoolEnabled:bool=Get bIsJumping(), Translation:Vector="0,0,0", Rotation="0,-40,0", Scale:Vector="1,1,1") -> (Pose) {
    alphaInputType=Bool
    rotationMode=BMM_Additive
    rotationSpace=BCS_BoneSpace
    alphaBoolBlend {blendInTime=0.2, blendOutTime=0.8}
    boneToModify {boneName=dyn_hair_back_01}
    rotation {yaw=-40}
   }
   n5 = Trail controller - Bone: dyn_hair_back_04(ComponentPose=@n4, Alpha=1) {
    bUsePlanarLimit=true
    chainLength=6
    debugLifeTime=0.1
    maxDeltaTime=0.0333
    relaxationSpeedScale=2
    alphaScaleBiasClamp {bInterpResult=true}
    rotationLimits [
     {limitMax=<180 180 180>}
     {limitMax=<180 180 180>}
     {limitMax=<180 180 180>}
     {limitMax=<180 180 180>}
     {limitMax=<180 180 180>}
     {limitMax=<180 180 180>}
    ]
    rotationOffsets [<0 0 0>, <0 0 0>, <0 0 0>, <0 0 0>, <0 0 0>, <0 0 0>]
    trailBone {boneName=dyn_hair_back_04}
    trailRelaxationSpeed {editorCurveData {keys [<at 0 15 RCIM_Linear>,<at 0.5289 13.1838 RCIM_Linear>,<at 1 6 RCIM_Linear>]}}
   }
   Save cached pose 'trailCache'(Pose=Component To Local(ComponentPose=@n5)) {cacheName=trailCache}
   n6 = Anim Dynamics - Bone: dyn_ribbon_side_00_lf(ComponentPose=Local To Component(LocalPose=Use cached pose 'ControlRig_Setup'()), Alpha=1, ExternalForce="0,0,0") {
    angularDampingOverride=1
    angularSpringConstant=20000
    bAngularSpring=true
    bOverrideAngularDamping=true
    bUseGravityOverride=true
    gravityScale=0.35
    simulationSpace=BoneRelative
    boundBone {boneName=dyn_ribbon_side_00_lf}
    componentLinearAccScale=<1 1 1>
    componentLinearVelScale=<1 1 1>
    gravityOverride {z=-980}
    physicsBodyDefinitions [{
      boundBone {boneName=dyn_ribbon_side_00_lf}
      boxExtents=<20 30 30>
      constraintSetup {angularLimitsMax=<0 0 0>, angularLimitsMin=<0 0 -45>, angularTarget=<1 0 0>}
      localJointOffset=<4 0 0>
     }]
    relativeSpaceBone {boneName=root}
    simSpaceSettings {maxAngularAcceleration=40, maxAngularVelocity=8, simSpaceAngularAlpha=1, externalAngularVelocity=<1 1 1>}
   }
   n7 = Anim Dynamics - Bone: dyn_ribbon_side_01_lf(ComponentPose=@n6, Alpha=1, ExternalForce="0,0,0") {
    angularDampingOverride=1
    angularSpringConstant=20000
    bAngularSpring=true
    bOverrideAngularDamping=true
    bUseGravityOverride=true
    gravityScale=0.35
    simulationSpace=BoneRelative
    boundBone {boneName=dyn_ribbon_side_01_lf}
    componentLinearAccScale=<1 1 1>
    componentLinearVelScale=<1 1 1>
    gravityOverride {z=-980}
    physicsBodyDefinitions [{
      boundBone {boneName=dyn_ribbon_side_01_lf}
      boxExtents=<20 30 30>
      constraintSetup {angularLimitsMax=<0 0 0>, angularLimitsMin=<0 0 -60>, angularTarget=<1 0 0>}
      localJointOffset=<4 0 0>
     }]
    relativeSpaceBone {boneName=root}
    simSpaceSettings {maxAngularAcceleration=40, maxAngularVelocity=8, simSpaceAngularAlpha=1, externalAngularVelocity=<1 1 1>}
   }
   n8 = Anim Dynamics - Bone: dyn_lace_side_00_lf(ComponentPose=@n7, Alpha=1, ExternalForce="0,0,0") {
    angularDampingOverride=1
    angularSpringConstant=20000
    bAngularSpring=true
    bOverrideAngularDamping=true
    bUseGravityOverride=true
    gravityScale=0.35
    simulationSpace=BoneRelative
    boundBone {boneName=dyn_lace_side_00_lf}
    componentLinearAccScale=<1 1 1>
    componentLinearVelScale=<1 1 1>
    gravityOverride {z=-980}
    physicsBodyDefinitions [{
      boundBone {boneName=dyn_lace_side_00_lf}
      boxExtents=<20 30 30>
      constraintSetup {angularLimitsMax=<0 0 0>, angularLimitsMin=<0 0 -45>, angularTarget=<1 0 0>}
      localJointOffset=<4 0 0>
     }]
    relativeSpaceBone {boneName=root}
    simSpaceSettings {maxAngularAcceleration=40, maxAngularVelocity=8, simSpaceAngularAlpha=1, externalAngularVelocity=<1 1 1>}
   }
   n9 = Anim Dynamics - Bone: dyn_lace_side_01_lf(ComponentPose=@n8, Alpha=1, ExternalForce="0,0,0") {
    angularDampingOverride=1
    angularSpringConstant=20000
    bAngularSpring=true
    bOverrideAngularDamping=true
    bUseGravityOverride=true
    gravityScale=0.35
    simulationSpace=BoneRelative
    boundBone {boneName=dyn_lace_side_01_lf}
    componentLinearAccScale=<1 1 1>
    componentLinearVelScale=<1 1 1>
    gravityOverride {z=-980}
    physicsBodyDefinitions [{
      boundBone {boneName=dyn_lace_side_01_lf}
      boxExtents=<20 30 30>
      constraintSetup {angularLimitsMax=<0 0 0>, angularLimitsMin=<0 0 -60>, angularTarget=<1 0 0>}
      localJointOffset=<4 0 0>
     }]
    relativeSpaceBone {boneName=root}
    simSpaceSettings {maxAngularAcceleration=40, maxAngularVelocity=8, simSpaceAngularAlpha=1, externalAngularVelocity=<1 1 1>}
   }
   n10 = Anim Dynamics - Bone: dyn_lace_side_02_lf(ComponentPose=@n9, Alpha=1, ExternalForce="0,0,0") {
    angularDampingOverride=1
    angularSpringConstant=20000
    bAngularSpring=true
    bOverrideAngularDamping=true
    bUseGravityOverride=true
    gravityScale=0.35
    simulationSpace=BoneRelative
    boundBone {boneName=dyn_lace_side_02_lf}
    componentLinearAccScale=<1 1 1>
    componentLinearVelScale=<1 1 1>
    gravityOverride {z=-980}
    physicsBodyDefinitions [{
      boundBone {boneName=dyn_lace_side_02_lf}
      boxExtents=<20 30 30>
      constraintSetup {angularLimitsMax=<0 0 0>, angularLimitsMin=<0 0 -60>, angularTarget=<1 0 0>}
      localJointOffset=<4 0 0>
     }]
    relativeSpaceBone {boneName=root}
    simSpaceSettings {maxAngularAcceleration=40, maxAngularVelocity=8, simSpaceAngularAlpha=1, externalAngularVelocity=<1 1 1>}
   }
   n11 = Anim Dynamics - Bone: dyn_ribbon_side_00_rt(ComponentPose=@n10, Alpha=1, ExternalForce="0,0,0") {
    angularDampingOverride=1
    angularSpringConstant=20000
    bAngularSpring=true
    bOverrideAngularDamping=true
    bUseGravityOverride=true
    gravityScale=0.35
    simulationSpace=BoneRelative
    boundBone {boneName=dyn_ribbon_side_00_rt}
    componentLinearAccScale=<1 1 1>
    componentLinearVelScale=<1 1 1>
    gravityOverride {z=-980}
    physicsBodyDefinitions [{
      boundBone {boneName=dyn_ribbon_side_00_rt}
      boxExtents=<20 30 30>
      constraintSetup {angularLimitsMax=<0 0 0>, angularLimitsMin=<0 0 -45>, angularTarget=<1 0 0>}
      localJointOffset=<4 0 0>
     }]
    relativeSpaceBone {boneName=root}
    simSpaceSettings {maxAngularAcceleration=40, maxAngularVelocity=8, simSpaceAngularAlpha=1, externalAngularVelocity=<1 1 1>}
   }
   n12 = Anim Dynamics - Bone: dyn_ribbon_side_01_rt(ComponentPose=@n11, Alpha=1, ExternalForce="0,0,0") {
    angularDampingOverride=1
    angularSpringConstant=20000
    bAngularSpring=true
    bOverrideAngularDamping=true
    bUseGravityOverride=true
    gravityScale=0.35
    simulationSpace=BoneRelative
    boundBone {boneName=dyn_ribbon_side_01_rt}
    componentLinearAccScale=<1 1 1>
    componentLinearVelScale=<1 1 1>
    gravityOverride {z=-980}
    physicsBodyDefinitions [{
      boundBone {boneName=dyn_ribbon_side_01_rt}
      boxExtents=<20 30 30>
      constraintSetup {angularLimitsMax=<0 0 0>, angularLimitsMin=<0 0 -60>, angularTarget=<1 0 0>}
      localJointOffset=<4 0 0>
     }]
    relativeSpaceBone {boneName=root}
    simSpaceSettings {maxAngularAcceleration=40, maxAngularVelocity=8, simSpaceAngularAlpha=1, externalAngularVelocity=<1 1 1>}
   }
   n13 = Anim Dynamics - Bone: dyn_lace_side_00_rt(ComponentPose=@n12, Alpha=1, ExternalForce="0,0,0") {
    angularDampingOverride=1
    angularSpringConstant=20000
    bAngularSpring=true
    bOverrideAngularDamping=true
    bUseGravityOverride=true
    gravityScale=0.35
    simulationSpace=BoneRelative
    boundBone {boneName=dyn_lace_side_00_rt}
    componentLinearAccScale=<1 1 1>
    componentLinearVelScale=<1 1 1>
    gravityOverride {z=-980}
    physicsBodyDefinitions [{
      boundBone {boneName=dyn_lace_side_00_rt}
      boxExtents=<20 30 30>
      constraintSetup {angularLimitsMax=<0 0 0>, angularLimitsMin=<0 0 -45>, angularTarget=<1 0 0>}
      localJointOffset=<4 0 0>
     }]
    relativeSpaceBone {boneName=root}
    simSpaceSettings {maxAngularAcceleration=40, maxAngularVelocity=8, simSpaceAngularAlpha=1, externalAngularVelocity=<1 1 1>}
   }
   n14 = Anim Dynamics - Bone: dyn_lace_side_01_rt(ComponentPose=@n13, Alpha=1, ExternalForce="0,0,0") {
    angularDampingOverride=1
    angularSpringConstant=20000
    bAngularSpring=true
    bOverrideAngularDamping=true
    bUseGravityOverride=true
    gravityScale=0.35
    simulationSpace=BoneRelative
    boundBone {boneName=dyn_lace_side_01_rt}
    componentLinearAccScale=<1 1 1>
    componentLinearVelScale=<1 1 1>
    gravityOverride {z=-980}
    physicsBodyDefinitions [{
      boundBone {boneName=dyn_lace_side_01_rt}
      boxExtents=<20 30 30>
      constraintSetup {angularLimitsMax=<0 0 0>, angularLimitsMin=<0 0 -60>, angularTarget=<1 0 0>}
      localJointOffset=<4 0 0>
     }]
    relativeSpaceBone {boneName=root}
    simSpaceSettings {maxAngularAcceleration=40, maxAngularVelocity=8, simSpaceAngularAlpha=1, externalAngularVelocity=<1 1 1>}
   }
   n15 = Anim Dynamics - Bone: dyn_lace_side_02_rt(ComponentPose=@n14, Alpha=1, ExternalForce="0,0,0") {
    angularDampingOverride=1
    angularSpringConstant=20000
    bAngularSpring=true
    bOverrideAngularDamping=true
    bUseGravityOverride=true
    gravityScale=0.35
    simulationSpace=BoneRelative
    boundBone {boneName=dyn_lace_side_02_rt}
    componentLinearAccScale=<1 1 1>
    componentLinearVelScale=<1 1 1>
    gravityOverride {z=-980}
    physicsBodyDefinitions [{
      boundBone {boneName=dyn_lace_side_02_rt}
      boxExtents=<20 30 30>
      constraintSetup {angularLimitsMax=<0 0 0>, angularLimitsMin=<0 0 -60>, angularTarget=<1 0 0>}
      localJointOffset=<4 0 0>
     }]
    relativeSpaceBone {boneName=root}
    simSpaceSettings {maxAngularAcceleration=40, maxAngularVelocity=8, simSpaceAngularAlpha=1, externalAngularVelocity=<1 1 1>}
   }
   Save cached pose 'AnimDynamics'(Pose=Component To Local(ComponentPose=@n15)) {cacheName=AnimDynamics}
   n16 = RigidBody!(
    ComponentPose=Local To Component(LocalPose=Use cached pose 'IsWearingCape Cache'())
    Alpha
    OverrideWorldGravity:Vector="0,0,0"
    ExternalForce
    SimSpaceSettings:Sim Space Settings Structure=(WorldAlpha=0.5,VelocityScaleZ=0.3,DampingAlpha=1,MaxLinearVelocity=700,MaxAngularVelocity=8,MaxLinearAcceleration=3500,MaxAngularAcceleration=40,ExternalLinearDragV=(X=0.2,Y=0.2,Z=0.3),ExternalLinearVelocity=(X=0.2,Y=0.2,Z=0.3),ExternalAngularVelocity=(X=0,Y=0,Z=0))
   ) -> (Pose) {
    bOverrideWorldGravity=true
    overridePhysicsAsset=&/F_MED_ShovelCrop_FaceAcc_FrontHair_Physics
    simulationSpace=BaseBoneSpace
    baseBoneRef {boneName=head}
    simSpaceSettings {
     maxAngularAcceleration=40
     maxAngularVelocity=8
     maxLinearAcceleration=3500
     maxLinearVelocity=700
     velocityScaleZ=0.3
     worldAlpha=0.5
     externalLinearDragV=<0.2 0.2 0.3>
     externalLinearVelocity=<0.2 0.2 0.3>
    }
   }
   n17 = RigidBody(
    ComponentPose=Local To Component(LocalPose=Use cached pose 'IsWearingCape Cache'())
    OverrideWorldGravity="0,0,0"
    SimSpaceSettings=(WorldAlpha=0.5,VelocityScaleZ=0.3,DampingAlpha=1,MaxLinearVelocity=700,MaxAngularVelocity=8,MaxLinearAcceleration=3500,MaxAngularAcceleration=40,ExternalLinearDragV=(X=0.2,Y=0.2,Z=0.3),ExternalLinearVelocity=(X=0.2,Y=0.2,Z=0.3),ExternalAngularVelocity=(X=0,Y=0,Z=0))
   ) {
    bOverrideWorldGravity=true
    overridePhysicsAsset=&/F_MED_ShovelCrop_FaceAcc_FrontHair_Cape_Physics
    simulationSpace=BaseBoneSpace
    baseBoneRef {boneName=head}
    simSpaceSettings {
     maxAngularAcceleration=40
     maxAngularVelocity=8
     maxLinearAcceleration=3500
     maxLinearVelocity=700
     velocityScaleZ=0.3
     worldAlpha=0.5
     externalLinearDragV=<0.2 0.2 0.3>
     externalLinearVelocity=<0.2 0.2 0.3>
    }
   }
   n18 = "Transform (Modify) Bone - Bone: dyn_hair_front_right_01_R"(ComponentPose=@n17, Alpha=1, Translation="0,0,0", Rotation="0,8,0", Scale="1,1,1") {
    rotationMode=BMM_Additive
    rotationSpace=BCS_BoneSpace
    translationMode=BMM_Additive
    translationSpace=BCS_BoneSpace
    boneToModify {boneName=dyn_hair_front_right_01_R}
    rotation {yaw=8}
   }
   n19 = "Transform (Modify) Bone - Bone: dyn_hair_front_right_01_L"!(ComponentPose=@n18, bAlphaBoolEnabled:bool=Get bRigidBodyEnabled(), Translation="0,0,0", Rotation="0,-10,0", Scale="1,1,1") -> (Pose) {
    alphaInputType=Bool
    rotationMode=BMM_Additive
    rotationSpace=BCS_BoneSpace
    boneToModify {boneName=dyn_hair_front_right_01_L}
    rotation {yaw=-10}
   }
   n20 = Get bClothEnabled()
   n21 = "Transform (Modify) Bone - Bone: offset_hair_back_01"!(
    ComponentPose="Transform (Modify) Bone - Bone: base_hair_back_00"(ComponentPose=@n19, Alpha=1, Translation="0,0,0", Rotation="0,5,0", Scale="1,1,1") {rotationMode=BMM_Additive, rotationSpace=BCS_BoneSpace, boneToModify {boneName=base_hair_back_00}, rotation {yaw=5}}
    bAlphaBoolEnabled:bool=@n20
    Translation="0,0,0"
    Rotation="0,5,0"
    Scale="1,1,1"
   ) -> (Pose) {
    alphaInputType=Bool
    rotationMode=BMM_Additive
    rotationSpace=BCS_BoneSpace
    translationMode=BMM_Additive
    translationSpace=BCS_BoneSpace
    boneToModify {boneName=offset_hair_back_01}
    rotation {yaw=5}
   }
   n22 = "Transform (Modify) Bone - Bone: dyn_hair_back_04"!(ComponentPose=@n21, bAlphaBoolEnabled:bool=@n20, Translation="0,0,0", Rotation="0,-10,0", Scale="1,1,1") -> (Pose) {
    alphaInputType=Bool
    rotationMode=BMM_Additive
    rotationSpace=BCS_BoneSpace
    translationMode=BMM_Additive
    translationSpace=BCS_BoneSpace
    boneToModify {boneName=dyn_hair_back_04}
    rotation {yaw=-10}
   }
   n23 = "Transform (Modify) Bone - Bone: offset_hair_back_01"(ComponentPose=@n22, bAlphaBoolEnabled=Get bIsSkyDiving(), Translation="0,0,0", Rotation="0,0,20", Scale="1,1,1") {
    alphaInputType=Bool
    rotationMode=BMM_Additive
    alphaBoolBlend {blendOption=QuadraticInOut}
    boneToModify {boneName=offset_hair_back_01}
    rotation {roll=20}
   }
   Save cached pose 'Rigidbody_Dynamics_1'(Pose=Blend Poses by bool(BlendPose_0=Component To Local(ComponentPose=@n23), BlendPose_1=Component To Local(ComponentPose=@n16), bActiveValue=Get isShovelCrop_Cape())) {cacheName=Rigidbody_Dynamics_1}
   n24 = Anim Dynamics - Bone: dyn_hair_back_00_a!(
    ComponentPose=Local To Component(LocalPose=Use cached pose 'Rigidbody_Dynamics_1'())
    Alpha=1
    GravityOverride:Vector="0,0,-980"
    AngularSpringConstant:float=3000
    ExternalForce="0,0,0"
   ) -> (Pose) {
    angularDampingOverride=1
    angularSpringConstant=3000
    bAngularSpring=true
    bOverrideAngularDamping=true
    bUseGravityOverride=true
    bUseSphericalLimits=true
    gravityScale=0.35
    simulationSpace=BoneRelative
    boundBone {boneName=dyn_hair_back_00_a}
    componentAppliedLinearAccClamp=<3000 3000 5000>
    componentLinearAccScale=<0.7 0.7 0.7>
    componentLinearVelScale=<0.7 0.7 0.7>
    gravityOverride {z=-980}
    physicsBodyDefinitions [{
      boundBone {boneName=dyn_hair_back_00_a}
      boxExtents=<20 30 30>
      constraintSetup {angularLimitsMax=<0 5 5>, angularLimitsMin=<0 -5 -30>, angularTarget=<1 0 0>}
      localJointOffset=<4 0 0>
     }]
    relativeSpaceBone {boneName=root}
    simSpaceSettings {maxAngularAcceleration=40, maxAngularVelocity=8}
    sphericalLimits [
     {limitRadius=16, drivingBone {boneName=pelvis}, sphereLocalOffset=<0 -2 0>}
     {limitRadius=15, drivingBone {boneName=upperarm_r}, sphereLocalOffset=<0 0 0>}
     {limitRadius=15, drivingBone {boneName=upperarm_l}, sphereLocalOffset=<0 0 0>}
    ]
   }
   n25 = Anim Dynamics - Bone: dyn_hair_back_01_a!(ComponentPose=@n24, Alpha=1, GravityOverride:Vector="0,0,-980", AngularSpringConstant:float=3000, ExternalForce="0,0,0") -> (Pose) {
    angularDampingOverride=1
    angularSpringConstant=3000
    bAngularSpring=true
    bOverrideAngularDamping=true
    bUseGravityOverride=true
    bUseSphericalLimits=true
    gravityScale=0.35
    simulationSpace=BoneRelative
    boundBone {boneName=dyn_hair_back_01_a}
    componentAppliedLinearAccClamp=<3000 3000 5000>
    componentLinearAccScale=<0.7 0.7 0.7>
    componentLinearVelScale=<0.7 0.7 0.7>
    gravityOverride {z=-980}
    physicsBodyDefinitions [{
      boundBone {boneName=dyn_hair_back_01_a}
      boxExtents=<20 30 30>
      constraintSetup {angularLimitsMax=<0 5 5>, angularLimitsMin=<0 -5 -30>, angularTarget=<1 0 0>}
      localJointOffset=<4 0 0>
     }]
    relativeSpaceBone {boneName=root}
    simSpaceSettings {maxAngularAcceleration=40, maxAngularVelocity=8}
    sphericalLimits [
     {limitRadius=16, drivingBone {boneName=pelvis}, sphereLocalOffset=<0 -2 0>}
     {limitRadius=15, drivingBone {boneName=upperarm_r}, sphereLocalOffset=<0 0 0>}
     {limitRadius=15, drivingBone {boneName=upperarm_l}, sphereLocalOffset=<0 0 0>}
    ]
   }
   n26 = Anim Dynamics - Bone: dyn_hair_back_02_a!(ComponentPose=@n25, Alpha=1, GravityOverride:Vector="0,0,-980", AngularSpringConstant:float=3000, ExternalForce="0,0,0") -> (Pose) {
    angularDampingOverride=1
    angularSpringConstant=3000
    bAngularSpring=true
    bOverrideAngularDamping=true
    bUseGravityOverride=true
    bUseSphericalLimits=true
    gravityScale=0.35
    simulationSpace=BoneRelative
    boundBone {boneName=dyn_hair_back_02_a}
    componentAppliedLinearAccClamp=<3000 3000 5000>
    componentLinearAccScale=<0.7 0.7 0.7>
    componentLinearVelScale=<0.7 0.7 0.7>
    gravityOverride {z=-980}
    physicsBodyDefinitions [{
      boundBone {boneName=dyn_hair_back_02_a}
      boxExtents=<20 30 30>
      constraintSetup {angularLimitsMax=<0 5 5>, angularLimitsMin=<0 -5 -30>, angularTarget=<1 0 0>}
      localJointOffset=<4 0 0>
     }]
    relativeSpaceBone {boneName=root}
    simSpaceSettings {maxAngularAcceleration=40, maxAngularVelocity=8}
    sphericalLimits [
     {limitRadius=16, drivingBone {boneName=pelvis}, sphereLocalOffset=<0 -2 0>}
     {limitRadius=15, drivingBone {boneName=upperarm_r}, sphereLocalOffset=<0 0 0>}
     {limitRadius=15, drivingBone {boneName=upperarm_l}, sphereLocalOffset=<0 0 0>}
    ]
   }
   n27 = Anim Dynamics - Bone: dyn_hair_back_03_a!(ComponentPose=@n26, Alpha=1, GravityOverride:Vector="0,0,-980", AngularSpringConstant:float=3000, ExternalForce="0,0,0") -> (Pose) {
    angularDampingOverride=1
    angularSpringConstant=3000
    bAngularSpring=true
    bOverrideAngularDamping=true
    bUseGravityOverride=true
    bUseSphericalLimits=true
    gravityScale=0.35
    simulationSpace=BoneRelative
    boundBone {boneName=dyn_hair_back_03_a}
    componentAppliedLinearAccClamp=<3000 3000 5000>
    componentLinearAccScale=<0.7 0.7 0.7>
    componentLinearVelScale=<0.7 0.7 0.7>
    gravityOverride {z=-980}
    physicsBodyDefinitions [{
      boundBone {boneName=dyn_hair_back_03_a}
      boxExtents=<20 30 30>
      constraintSetup {angularLimitsMax=<0 5 5>, angularLimitsMin=<0 -5 -30>, angularTarget=<1 0 0>}
      localJointOffset=<4 0 0>
     }]
    relativeSpaceBone {boneName=root}
    simSpaceSettings {maxAngularAcceleration=40, maxAngularVelocity=8}
    sphericalLimits [
     {limitRadius=16, drivingBone {boneName=pelvis}, sphereLocalOffset=<0 -2 0>}
     {limitRadius=15, drivingBone {boneName=upperarm_r}, sphereLocalOffset=<0 0 0>}
     {limitRadius=15, drivingBone {boneName=upperarm_l}, sphereLocalOffset=<0 0 0>}
    ]
   }
   n28 = Anim Dynamics - Bone: dyn_hair_back_00_c!(ComponentPose=@n27, Alpha=1, GravityOverride:Vector="0,0,-980", AngularSpringConstant:float=3000, ExternalForce="0,0,0") -> (Pose) {
    angularDampingOverride=1
    angularSpringConstant=3000
    bAngularSpring=true
    bOverrideAngularDamping=true
    bUseGravityOverride=true
    bUseSphericalLimits=true
    gravityScale=0.35
    simulationSpace=BoneRelative
    boundBone {boneName=dyn_hair_back_00_c}
    componentAppliedLinearAccClamp=<3000 3000 5000>
    componentLinearAccScale=<0.7 0.7 0.7>
    componentLinearVelScale=<0.7 0.7 0.7>
    gravityOverride {z=-980}
    physicsBodyDefinitions [{
      boundBone {boneName=dyn_hair_back_00_c}
      boxExtents=<20 30 30>
      constraintSetup {angularLimitsMax=<0 5 5>, angularLimitsMin=<0 -5 -30>, angularTarget=<1 0 0>}
      localJointOffset=<4 0 0>
     }]
    relativeSpaceBone {boneName=root}
    simSpaceSettings {maxAngularAcceleration=40, maxAngularVelocity=8}
    sphericalLimits [
     {limitRadius=16, drivingBone {boneName=pelvis}, sphereLocalOffset=<0 -2 0>}
     {limitRadius=15, drivingBone {boneName=upperarm_r}, sphereLocalOffset=<0 0 0>}
     {limitRadius=15, drivingBone {boneName=upperarm_l}, sphereLocalOffset=<0 0 0>}
    ]
   }
   n29 = Anim Dynamics - Bone: dyn_hair_back_01_c!(ComponentPose=@n28, Alpha=1, GravityOverride:Vector="0,0,-980", AngularSpringConstant:float=3000, ExternalForce="0,0,0") -> (Pose) {
    angularDampingOverride=1
    angularSpringConstant=3000
    bAngularSpring=true
    bOverrideAngularDamping=true
    bUseGravityOverride=true
    bUseSphericalLimits=true
    gravityScale=0.35
    simulationSpace=BoneRelative
    boundBone {boneName=dyn_hair_back_01_c}
    componentAppliedLinearAccClamp=<3000 3000 5000>
    componentLinearAccScale=<0.7 0.7 0.7>
    componentLinearVelScale=<0.7 0.7 0.7>
    gravityOverride {z=-980}
    physicsBodyDefinitions [{
      boundBone {boneName=dyn_hair_back_01_c}
      boxExtents=<20 30 30>
      constraintSetup {angularLimitsMax=<0 5 5>, angularLimitsMin=<0 -5 -30>, angularTarget=<1 0 0>}
      localJointOffset=<4 0 0>
     }]
    relativeSpaceBone {boneName=root}
    simSpaceSettings {maxAngularAcceleration=40, maxAngularVelocity=8}
    sphericalLimits [
     {limitRadius=16, drivingBone {boneName=pelvis}, sphereLocalOffset=<0 -2 0>}
     {limitRadius=15, drivingBone {boneName=upperarm_r}, sphereLocalOffset=<0 0 0>}
     {limitRadius=15, drivingBone {boneName=upperarm_l}, sphereLocalOffset=<0 0 0>}
    ]
   }
   n30 = Anim Dynamics - Bone: dyn_hair_back_02_c!(ComponentPose=@n29, Alpha=1, GravityOverride:Vector="0,0,-980", AngularSpringConstant:float=3000, ExternalForce="0,0,0") -> (Pose) {
    angularDampingOverride=1
    angularSpringConstant=3000
    bAngularSpring=true
    bOverrideAngularDamping=true
    bUseGravityOverride=true
    bUseSphericalLimits=true
    gravityScale=0.35
    simulationSpace=BoneRelative
    boundBone {boneName=dyn_hair_back_02_c}
    componentAppliedLinearAccClamp=<3000 3000 5000>
    componentLinearAccScale=<0.7 0.7 0.7>
    componentLinearVelScale=<0.7 0.7 0.7>
    gravityOverride {z=-980}
    physicsBodyDefinitions [{
      boundBone {boneName=dyn_hair_back_02_c}
      boxExtents=<20 30 30>
      constraintSetup {angularLimitsMax=<0 5 5>, angularLimitsMin=<0 -5 -30>, angularTarget=<1 0 0>}
      localJointOffset=<4 0 0>
     }]
    relativeSpaceBone {boneName=root}
    simSpaceSettings {maxAngularAcceleration=40, maxAngularVelocity=8}
    sphericalLimits [
     {limitRadius=16, drivingBone {boneName=pelvis}, sphereLocalOffset=<0 -2 0>}
     {limitRadius=15, drivingBone {boneName=upperarm_r}, sphereLocalOffset=<0 0 0>}
     {limitRadius=15, drivingBone {boneName=upperarm_l}, sphereLocalOffset=<0 0 0>}
    ]
   }
   n31 = Anim Dynamics - Bone: dyn_hair_back_00_d!(
    ComponentPose=@n30
    Alpha=1
    GravityOverride:Vector="0,0,-980"
    AngularSpringConstant:float=3000
    ComponentLinearAccScale:Vector="1,1,1"
    ExternalForce="0,0,0"
   ) -> (Pose) {
    angularDampingOverride=1
    angularSpringConstant=3000
    bAngularSpring=true
    bOverrideAngularDamping=true
    bUseGravityOverride=true
    bUseSphericalLimits=true
    gravityScale=0.35
    simulationSpace=BoneRelative
    boundBone {boneName=dyn_hair_back_00_d}
    componentAppliedLinearAccClamp=<3000 3000 5000>
    componentLinearAccScale=<0.7 0.7 0.7>
    componentLinearVelScale=<0.7 0.7 0.7>
    gravityOverride {z=-980}
    physicsBodyDefinitions [{
      boundBone {boneName=dyn_hair_back_00_d}
      boxExtents=<20 30 30>
      constraintSetup {angularLimitsMax=<0 5 5>, angularLimitsMin=<0 -5 0>, angularTarget=<1 0 0>}
      localJointOffset=<4 0 0>
     }]
    relativeSpaceBone {boneName=root}
    simSpaceSettings {maxAngularAcceleration=40, maxAngularVelocity=8}
    sphericalLimits [
     {limitRadius=16, drivingBone {boneName=pelvis}, sphereLocalOffset=<0 -2 0>}
     {limitRadius=15, drivingBone {boneName=upperarm_r}, sphereLocalOffset=<0 0 0>}
     {limitRadius=15, drivingBone {boneName=upperarm_l}, sphereLocalOffset=<0 0 0>}
    ]
   }
   n32 = Anim Dynamics - Bone: dyn_hair_back_01_d!(ComponentPose=@n31, Alpha=1, GravityOverride:Vector="0,0,-980", AngularSpringConstant:float=3000, ExternalForce="0,0,0") -> (Pose) {
    angularDampingOverride=1
    angularSpringConstant=3000
    bAngularSpring=true
    bOverrideAngularDamping=true
    bUseGravityOverride=true
    bUseSphericalLimits=true
    gravityScale=0.35
    simulationSpace=BoneRelative
    boundBone {boneName=dyn_hair_back_01_d}
    componentAppliedLinearAccClamp=<3000 3000 5000>
    componentLinearAccScale=<0.7 0.7 0.7>
    componentLinearVelScale=<0.7 0.7 0.7>
    gravityOverride {z=-980}
    physicsBodyDefinitions [{
      boundBone {boneName=dyn_hair_back_01_d}
      boxExtents=<20 30 30>
      constraintSetup {angularLimitsMax=<0 5 5>, angularLimitsMin=<0 -5 0>, angularTarget=<1 0 0>}
      localJointOffset=<4 0 0>
     }]
    relativeSpaceBone {boneName=root}
    simSpaceSettings {maxAngularAcceleration=40, maxAngularVelocity=8}
    sphericalLimits [
     {limitRadius=16, drivingBone {boneName=pelvis}, sphereLocalOffset=<0 -2 0>}
     {limitRadius=15, drivingBone {boneName=upperarm_r}, sphereLocalOffset=<0 0 0>}
     {limitRadius=15, drivingBone {boneName=upperarm_l}, sphereLocalOffset=<0 0 0>}
    ]
   }
   n33 = Anim Dynamics - Bone: dyn_hair_back_02_d!(ComponentPose=@n32, Alpha=1, GravityOverride:Vector="0,0,-980", AngularSpringConstant:float=3000, ExternalForce="0,0,0") -> (Pose) {
    angularDampingOverride=1
    angularSpringConstant=3000
    bAngularSpring=true
    bOverrideAngularDamping=true
    bUseGravityOverride=true
    bUseSphericalLimits=true
    gravityScale=0.35
    simulationSpace=BoneRelative
    boundBone {boneName=dyn_hair_back_02_d}
    componentAppliedLinearAccClamp=<3000 3000 5000>
    componentLinearAccScale=<0.7 0.7 0.7>
    componentLinearVelScale=<0.7 0.7 0.7>
    gravityOverride {z=-980}
    physicsBodyDefinitions [{
      boundBone {boneName=dyn_hair_back_02_d}
      boxExtents=<20 30 30>
      constraintSetup {angularLimitsMax=<0 5 5>, angularLimitsMin=<0 -5 0>, angularTarget=<1 0 0>}
      localJointOffset=<4 0 0>
     }]
    relativeSpaceBone {boneName=root}
    simSpaceSettings {maxAngularAcceleration=40, maxAngularVelocity=8}
    sphericalLimits [
     {limitRadius=16, drivingBone {boneName=pelvis}, sphereLocalOffset=<0 -2 0>}
     {limitRadius=15, drivingBone {boneName=spine_05}, sphereLocalOffset=<0 0 0>}
     {limitRadius=15, drivingBone {boneName=spine_03}, sphereLocalOffset=<0 0 0>}
    ]
   }
   n34 = Anim Dynamics - Bone: dyn_hair_back_03_d!(ComponentPose=@n33, Alpha=1, GravityOverride:Vector="0,0,-980", AngularSpringConstant:float=3000, ExternalForce="0,0,0") -> (Pose) {
    angularDampingOverride=1
    angularSpringConstant=3000
    bAngularSpring=true
    bOverrideAngularDamping=true
    bUseGravityOverride=true
    bUseSphericalLimits=true
    gravityScale=0.35
    simulationSpace=BoneRelative
    boundBone {boneName=dyn_hair_back_03_d}
    componentAppliedLinearAccClamp=<3000 3000 5000>
    componentLinearAccScale=<0.7 0.7 0.7>
    componentLinearVelScale=<0.7 0.7 0.7>
    gravityOverride {z=-980}
    physicsBodyDefinitions [{
      boundBone {boneName=dyn_hair_back_03_d}
      boxExtents=<20 30 30>
      constraintSetup {angularLimitsMax=<0 5 5>, angularLimitsMin=<0 -5 0>, angularTarget=<1 0 0>}
      localJointOffset=<4 0 0>
     }]
    relativeSpaceBone {boneName=root}
    simSpaceSettings {maxAngularAcceleration=40, maxAngularVelocity=8}
    sphericalLimits [
     {limitRadius=16, drivingBone {boneName=pelvis}, sphereLocalOffset=<0 -2 0>}
     {limitRadius=15, drivingBone {boneName=spine_05}, sphereLocalOffset=<0 0 0>}
     {limitRadius=15, drivingBone {boneName=spine_03}, sphereLocalOffset=<0 0 0>}
    ]
   }
   n35 = Anim Dynamics - Bone: dyn_hair_back_00_e!(ComponentPose=@n34, Alpha=1, GravityOverride:Vector="0,0,-980", AngularSpringConstant:float=3000, ExternalForce="0,0,0") -> (Pose) {
    angularDampingOverride=1
    angularSpringConstant=3000
    bAngularSpring=true
    bOverrideAngularDamping=true
    bUseGravityOverride=true
    bUseSphericalLimits=true
    gravityScale=0.35
    simulationSpace=BoneRelative
    boundBone {boneName=dyn_hair_back_00_e}
    componentAppliedLinearAccClamp=<3000 3000 5000>
    componentLinearAccScale=<0.7 0.7 0.7>
    componentLinearVelScale=<0.7 0.7 0.7>
    gravityOverride {z=-980}
    physicsBodyDefinitions [{
      boundBone {boneName=dyn_hair_back_00_e}
      boxExtents=<20 30 30>
      constraintSetup {angularLimitsMax=<0 5 5>, angularLimitsMin=<0 -5 -30>, angularTarget=<1 0 0>}
      localJointOffset=<4 0 0>
     }]
    relativeSpaceBone {boneName=root}
    simSpaceSettings {maxAngularAcceleration=40, maxAngularVelocity=8}
    sphericalLimits [
     {limitRadius=16, drivingBone {boneName=pelvis}, sphereLocalOffset=<0 -2 0>}
     {limitRadius=15, drivingBone {boneName=spine_05}, sphereLocalOffset=<0 0 0>}
     {limitRadius=15, drivingBone {boneName=spine_03}, sphereLocalOffset=<0 0 0>}
    ]
   }
   n36 = Anim Dynamics - Bone: dyn_hair_back_01_e!(ComponentPose=@n35, Alpha=1, GravityOverride:Vector="0,0,-980", AngularSpringConstant:float=3000, ExternalForce="0,0,0") -> (Pose) {
    angularDampingOverride=1
    angularSpringConstant=3000
    bAngularSpring=true
    bOverrideAngularDamping=true
    bUseGravityOverride=true
    bUseSphericalLimits=true
    gravityScale=0.35
    simulationSpace=BoneRelative
    boundBone {boneName=dyn_hair_back_01_e}
    componentAppliedLinearAccClamp=<3000 3000 5000>
    componentLinearAccScale=<0.7 0.7 0.7>
    componentLinearVelScale=<0.7 0.7 0.7>
    gravityOverride {z=-980}
    physicsBodyDefinitions [{
      boundBone {boneName=dyn_hair_back_01_e}
      boxExtents=<20 30 30>
      constraintSetup {angularLimitsMax=<0 5 5>, angularLimitsMin=<0 -5 -30>, angularTarget=<1 0 0>}
      localJointOffset=<4 0 0>
     }]
    relativeSpaceBone {boneName=root}
    simSpaceSettings {maxAngularAcceleration=40, maxAngularVelocity=8}
    sphericalLimits [
     {limitRadius=16, drivingBone {boneName=pelvis}, sphereLocalOffset=<0 -2 0>}
     {limitRadius=15, drivingBone {boneName=spine_05}, sphereLocalOffset=<0 0 0>}
     {limitRadius=15, drivingBone {boneName=spine_03}, sphereLocalOffset=<0 0 0>}
    ]
   }
   n37 = Anim Dynamics - Bone: dyn_hair_back_02_e!(ComponentPose=@n36, Alpha=1, GravityOverride:Vector="0,0,-980", AngularSpringConstant:float=3000, ExternalForce="0,0,0") -> (Pose) {
    angularDampingOverride=1
    angularSpringConstant=3000
    bAngularSpring=true
    bOverrideAngularDamping=true
    bUseGravityOverride=true
    bUseSphericalLimits=true
    gravityScale=0.35
    simulationSpace=BoneRelative
    boundBone {boneName=dyn_hair_back_02_e}
    componentAppliedLinearAccClamp=<3000 3000 5000>
    componentLinearAccScale=<0.7 0.7 0.7>
    componentLinearVelScale=<0.7 0.7 0.7>
    gravityOverride {z=-980}
    physicsBodyDefinitions [{
      boundBone {boneName=dyn_hair_back_02_e}
      boxExtents=<20 30 30>
      constraintSetup {angularLimitsMax=<0 5 5>, angularLimitsMin=<0 -5 -30>, angularTarget=<1 0 0>}
      localJointOffset=<4 0 0>
     }]
    relativeSpaceBone {boneName=root}
    simSpaceSettings {maxAngularAcceleration=40, maxAngularVelocity=8}
    sphericalLimits [
     {limitRadius=16, drivingBone {boneName=pelvis}, sphereLocalOffset=<0 -2 0>}
     {limitRadius=15, drivingBone {boneName=spine_05}, sphereLocalOffset=<0 0 0>}
     {limitRadius=15, drivingBone {boneName=spine_03}, sphereLocalOffset=<0 0 0>}
    ]
   }
   n38 = Anim Dynamics - Bone: dyn_hair_back_03_e!(ComponentPose=@n37, Alpha=1, GravityOverride:Vector="0,0,-980", AngularSpringConstant:float=3000, ExternalForce="0,0,0") -> (Pose) {
    angularDampingOverride=1
    angularSpringConstant=3000
    bAngularSpring=true
    bOverrideAngularDamping=true
    bUseGravityOverride=true
    bUseSphericalLimits=true
    gravityScale=0.35
    simulationSpace=BoneRelative
    boundBone {boneName=dyn_hair_back_03_e}
    componentAppliedLinearAccClamp=<3000 3000 5000>
    componentLinearAccScale=<0.7 0.7 0.7>
    componentLinearVelScale=<0.7 0.7 0.7>
    gravityOverride {z=-980}
    physicsBodyDefinitions [{
      boundBone {boneName=dyn_hair_back_03_e}
      boxExtents=<20 30 30>
      constraintSetup {angularLimitsMax=<0 5 5>, angularLimitsMin=<0 -5 -20>, angularTarget=<1 0 0>}
      localJointOffset=<4 0 0>
     }]
    relativeSpaceBone {boneName=root}
    simSpaceSettings {maxAngularAcceleration=40, maxAngularVelocity=8}
    sphericalLimits [
     {limitRadius=16, drivingBone {boneName=pelvis}, sphereLocalOffset=<0 -2 0>}
     {limitRadius=15, drivingBone {boneName=spine_05}, sphereLocalOffset=<0 0 0>}
     {limitRadius=15, drivingBone {boneName=spine_03}, sphereLocalOffset=<0 0 0>}
    ]
   }
   n39 = "Constraint - dyn_hair_back_01 (P)"(ComponentPose=Local To Component(LocalPose=Use cached pose 'Rigidbody_Dynamics_1'()), Alpha=1, ConstraintWeights_0=1) {boneToModify {boneName=dyn_hair_back_01}, constraintSetup [{targetBone {boneName=head}}], constraintWeights [1]}
   n40 = "Constraint - dyn_hair_back_02 (P)"(ComponentPose=@n39, Alpha=1, ConstraintWeights_0=1) {boneToModify {boneName=dyn_hair_back_02}, constraintSetup [{targetBone {boneName=spine_05}}], constraintWeights [1]}
   n41 = "Transform (Modify) Bone - Bone: dyn_hair_back_02"!(ComponentPose=@n40, Alpha=1, Translation:Vector="-0.144,-0.25,-0.461", Rotation="-0.088,10.051,0.079", Scale:Vector="1,0.8,1") -> (Pose) {
    rotationMode=BMM_Additive
    rotationSpace=BCS_BoneSpace
    scaleMode=BMM_Additive
    scaleSpace=BCS_BoneSpace
    translationMode=BMM_Additive
    translationSpace=BCS_BoneSpace
    boneToModify {boneName=dyn_hair_back_02}
    rotation=<rot -0.088 10.051 0.079>
    scale {y=0.8}
    translation=<-0.144 -0.25 -0.461>
   }
   n42 = "Transform (Modify) Bone - Bone: dyn_hair_back_03"(ComponentPose=@n41, Alpha=1, Translation="-0.15,6.69,0.183", Rotation="0.000,5.000,-0.000", Scale="1,0.7,1") {
    rotationMode=BMM_Additive
    rotationSpace=BCS_BoneSpace
    scaleMode=BMM_Additive
    scaleSpace=BCS_BoneSpace
    translationMode=BMM_Additive
    translationSpace=BCS_BoneSpace
    boneToModify {boneName=dyn_hair_back_03}
    rotation {yaw=5}
    scale {y=0.7}
    translation=<-0.15 6.69 0.183>
   }
   n43 = "Constraint - dyn_hair_back_03 (R)"(ComponentPose=@n42, Alpha=1, ConstraintWeights_0=1) {boneToModify {boneName=dyn_hair_back_03}, constraintSetup [{transformType=Rotation, targetBone {boneName=spine_03}}], constraintWeights [1]}
   n44 = "Transform (Modify) Bone - Bone: dyn_hair_back_04"!(ComponentPose=@n43, Alpha:float=1, Translation="0.613,-1.291,-1.415", Rotation="0,0,0", Scale="1,1,1") -> (Pose) {
    rotationMode=BMM_Additive
    rotationSpace=BCS_BoneSpace
    scaleMode=BMM_Additive
    scaleSpace=BCS_BoneSpace
    translationMode=BMM_Additive
    translationSpace=BCS_BoneSpace
    boneToModify {boneName=dyn_hair_back_04}
    translation=<0.613 -1.291 -1.415>
   }
   n45 = "Transform (Modify) Bone - Bone: dyn_hair_back_05"(ComponentPose=@n44, Alpha=1, Translation="-0.689,-3.957,-0.100", Rotation="0,0,0", Scale="1,1,1") {
    rotationMode=BMM_Additive
    rotationSpace=BCS_BoneSpace
    translationMode=BMM_Additive
    translationSpace=BCS_BoneSpace
    boneToModify {boneName=dyn_hair_back_05}
    translation=<-0.689 -3.957 -0.1>
   }
   n46 = "Constraint - dyn_hair_back_04 (R)"(ComponentPose=@n45, Alpha=1, ConstraintWeights_0=1) {boneToModify {boneName=dyn_hair_back_04}, constraintSetup [{transformType=Rotation, targetBone {boneName=spine_02}}], constraintWeights [1]}
   Save cached pose 'Rigidbody_Dynamics_2'(Pose=Blend Poses by bool(
     BlendPose_0=Component To Local(ComponentPose=@n46)
     BlendPose_1=Component To Local(ComponentPose=@n38)
     bActiveValue=Get Is Wearing Backpack()
    )) {cacheName=Rigidbody_Dynamics_2}
   n47 = RigidBody(
    ComponentPose=Local To Component(LocalPose=Use cached pose 'AnimDynamics'())
    OverrideWorldGravity="0,0,0"
    SimSpaceSettings=(WorldAlpha=1,VelocityScaleZ=0.5,DampingAlpha=0.5,MaxLinearVelocity=700,MaxAngularVelocity=8,MaxLinearAcceleration=3500,MaxAngularAcceleration=40,ExternalLinearDragV=(X=0.2,Y=0.2,Z=0.3),ExternalLinearVelocity=(X=0.2,Y=0.2,Z=0.3),ExternalAngularVelocity=(X=0,Y=0,Z=0))
   ) {
    bClampLinearTranslationLimitToRefPose=true
    bOverrideWorldGravity=true
    overridePhysicsAsset=&/F_MED_ShovelCrop_FaceAcc_Physics
    simulationSpace=BaseBoneSpace
    baseBoneRef {boneName=head}
    simSpaceSettings {
     dampingAlpha=0.5
     maxAngularAcceleration=40
     maxAngularVelocity=8
     maxLinearAcceleration=3500
     maxLinearVelocity=700
     velocityScaleZ=0.5
     worldAlpha=1
     externalLinearDragV=<0.2 0.2 0.3>
     externalLinearVelocity=<0.2 0.2 0.3>
    }
   }
   n48 = RigidBody(
    ComponentPose=Local To Component(LocalPose=Use cached pose 'AnimDynamics'())
    OverrideWorldGravity="0,0,0"
    SimSpaceSettings=(WorldAlpha=1,VelocityScaleZ=1,DampingAlpha=11,MaxLinearVelocity=700,MaxAngularVelocity=8,MaxLinearAcceleration=3500,MaxAngularAcceleration=40,ExternalLinearDragV=(X=1,Y=1,Z=1),ExternalLinearVelocity=(X=1,Y=1,Z=1),ExternalAngularVelocity=(X=0,Y=0,Z=0))
   ) {
    bClampLinearTranslationLimitToRefPose=true
    bOverrideWorldGravity=true
    overridePhysicsAsset=&/F_MED_ShovelCrop_FaceAcc_ShovelCape_Physics
    simulationSpace=BaseBoneSpace
    baseBoneRef {boneName=head}
    simSpaceSettings {
     dampingAlpha=11
     maxAngularAcceleration=40
     maxAngularVelocity=8
     maxLinearAcceleration=3500
     maxLinearVelocity=700
     worldAlpha=1
     externalLinearDragV=<1 1 1>
     externalLinearVelocity=<1 1 1>
    }
   }
   n49 = "Transform (Modify) Bone - Bone: base_hair_back_00"(ComponentPose=@n48, Alpha=Get moveForward() -> (moveForward:double), Translation="0,0,0", Rotation="0,-15,0", Scale="1,1,1") {rotationMode=BMM_Additive, rotationSpace=BCS_BoneSpace, boneToModify {boneName=base_hair_back_00}, rotation {yaw=-15}}
   n50 = "Transform (Modify) Bone - Bone: base_hair_back_00"!(ComponentPose=@n49, bAlphaBoolEnabled:bool=Get bRigidBodyEnabled(), Translation="0,0,0", Rotation="0,-5,0", Scale="1,1,1") -> (Pose) {alphaInputType=Bool, rotationMode=BMM_Additive, rotationSpace=BCS_BoneSpace, boneToModify {boneName=base_hair_back_00}, rotation {yaw=-5}}
   n51 = "Transform (Modify) Bone - Bone: offset_hair_back_01"(ComponentPose=@n50, bAlphaBoolEnabled=@n20, Translation="0,0,0", Rotation="0,-10,0", Scale="1,1,1") {
    alphaInputType=Bool
    rotationMode=BMM_Additive
    rotationSpace=BCS_BoneSpace
    translationMode=BMM_Additive
    translationSpace=BCS_BoneSpace
    boneToModify {boneName=offset_hair_back_01}
    rotation {yaw=-10}
   }
   n52 = "Transform (Modify) Bone - Bone: dyn_hair_back_04"!(ComponentPose=@n51, bAlphaBoolEnabled:bool=@n20, Translation="0,0,0", Rotation="0,10,0", Scale="1,1,1") -> (Pose) {
    alphaInputType=Bool
    rotationMode=BMM_Additive
    rotationSpace=BCS_BoneSpace
    translationMode=BMM_Additive
    translationSpace=BCS_BoneSpace
    boneToModify {boneName=dyn_hair_back_04}
    rotation {yaw=10}
   }
   Save cached pose 'IsWearingCape Cache'(Pose=Blend Poses by bool(
     BlendPose_0=Component To Local(ComponentPose=@n52)
     BlendPose_1=Component To Local(ComponentPose=@n47)
     bActiveValue=Get Is Wearing Cape() -> (Is Wearing Cape:bool)
    )) {cacheName=IsWearingCape Cache}
  ]}
 EventGraph {nodes [
   Event BlueprintUpdateAnimation()
   > n1 = Sequence!(execute) -> (then_0, then_1, then_2:Exec, then_3:Exec, then_4:Exec, then_5:Exec)
   > n2 = "Gravity Override State Function (Parameter Asset) (Nativized)"(
    CharacterPartAnimInstance=Self-Reference()
    GameTimeInSeconds=GetGameTimeInSeconds()
    Parameters=&/F_MED_ShovelCrop_FaceAcc_Parameters
   )
   > Set Gravity Override(Gravity Override=@n2)
   > Set RBAN Settings Out(RBAN Settings Out=@n2.OutRigidBodyAnimNodeParameters)
   n3 = PlayerMovement(execute=@n1.then_1, AnimInstance=Self-Reference(), XDotProductCoefficient=1, YDotProductCoefficient=1, ZDotProductCoefficient=1)
   n4 = BreakVector(InVec=MakeVector(X=@n3, Y=@n3.YMovement))
   > Set strafeLeft float(strafeLeft float:double="Absolute (Float)"(A="Clamp (Float)"(Value=@n4, Min=-1, Max=0))) -> (Output_Get:double)
   > Set strafeRight float(strafeRight float:double="Clamp (Float)"(Value=@n4, Min=0, Max=1)) -> (Output_Get:double)
   > Set moveForward(moveForward:double=float * float(A="Clamp (Float)"(Value=@n4.Y, Min=0, Max=1), B=Get iswalking() -> (iswalking:double)) {is=title K2Node_PromotableOperator}) -> (Output_Get:double)
   > Set moveBackward(moveBackward:double="Absolute (Float)"(A="Clamp (Float)"(Value=@n4.Y, Min=-1, Max=0))) -> (Output_Get:double)
   n5 = Branch(execute=@n1.then_2, Condition=OR Boolean(A=Get bIsSkydiveDiveMode(), B=Get bIsSkydiveFloating()))
   > Set angularSpringConstant(angularSpringConstant:double=2000) -> (Output_Get:double)
   n6 = Branch(execute=@n5.else, Condition=Get bIsSprinting())
   > n7 = Branch(Condition=Get bIsTacticalSprinting())
   > Set angularSpringConstant(angularSpringConstant=10000)
   Set angularSpringConstant(execute=@n7.else, angularSpringConstant=10000)
   Set angularSpringConstant(execute=@n6.else, angularSpringConstant=3000)
   n8 = Cast To FortPlayerPawn(execute=@n1.then_3, Object=TryGetPawnOwner())
   n9 = GetSkeletalMeshForPartType(self=@n8, PartType=Backpack)
   > n10 = Set isShovelCrop_Cape(isShovelCrop_Cape:bool=MatchesWildcard(SourceString=GetDisplayName(Object=@n9), Wildcard=*F_MED_ShovelCrop_Cape*, SearchCase=CaseSensitive)) -> (Output_Get:bool)
   n11 = CheckForBackbling()
   > n12 = Set Is Wearing Cape(Is Wearing Cape=OR Boolean(A=@n10, B=@n11.bIsBackblingACape))
   > n13 = Branch(Condition=OR Boolean!(A=@n11, B=false, C:bool=@n11.bIsBackblingLong) -> (ReturnValue))
   > n14 = Branch(Condition=OR Boolean!(A=@n12, B=@n10) -> (ReturnValue))
   > Set Is Wearing Backpack(Is Wearing Backpack:bool=false) -> (Output_Get:bool)
   Set Is Wearing Backpack(execute=@n14.else, Is Wearing Backpack=true)
   Set Is Wearing Backpack(execute=@n13.else, Is Wearing Backpack=false)
   n15 = PlayerMovement(execute=@n1.then_4, AnimInstance=Self-Reference(), XDotProductCoefficient=1, YDotProductCoefficient=1, ZDotProductCoefficient=1)
   n16 = MapRangeClamped(Value=@n15.YMovement, InRangeA=0, InRangeB=1)
   n17 = "Absolute (Float)"(A=@n15.ZMovement)
   n18 = "Absolute (Float)"(A=@n15)
   > Set BwdMovement(BwdMovement:double=MapRangeClamped(
     Value=float + float!(A=@n18, B=@n16, C:double=@n17) -> (ReturnValue) {is=title K2Node_CommutativeAssociativeBinaryOperator}
     InRangeA=0
     InRangeB=1.4
    )) -> (Output_Get:double)
   n19 = Branch(execute=@n1.then_5, Condition=Get bIsSwimSprinting() -> (bIsSwimSprinting:bool))
   > Set TrailControllerRot(TrailControllerRot:Rotator="15,0, 0") -> (Output_Get:Rotator)
   n20 = Branch(execute=@n19.else, Condition=Get bIsSkyDiving())
   > n21 = Branch(Condition=Get bIsParachuteOpened())
   > Set TrailControllerRot(TrailControllerRot="0,-25,0")
   n22 = Branch(execute=@n21.else, Condition="float > float"(A=Get PawnVelocityZ(), B=50) {is=title K2Node_CallFunction})
   > Set BwdMovement(BwdMovement=float * float(A=Get BwdMovement() -> (BwdMovement:double), B=0) {is=title K2Node_CommutativeAssociativeBinaryOperator})
   > Set TrailControllerRot(TrailControllerRot="-15,0,0")
   n23 = Branch(execute=@n22.else, Condition=Get bIsSkydiveDiveMode())
   > Set TrailControllerRot(TrailControllerRot="0,-25,0")
   > Set BwdMovement(BwdMovement=float * float(A=Get BwdMovement(), B=0))
   Set TrailControllerRot(execute=@n23.else, TrailControllerRot="0,-20,0")
   n24 = Branch(execute=@n20.else, Condition=Get bIsCrouching())
   > Set TrailControllerRot(TrailControllerRot="0,-5,0")
   Set TrailControllerRot(execute=@n24.else, TrailControllerRot="0,-5,0")
   > Set BwdMovement(BwdMovement=float * float(A=Get BwdMovement(), B=1))
   Event OnBeginCrouching()
   > Set isCrouching(isCrouching:double=1) -> (Output_Get:double)
   Event OnEndCrouching()
   > Set isCrouching(isCrouching=0)
   Event OnCharacterCustomizationLoaded()
   > Sequence!(execute) -> (then_0, then_1)
   > n25 = Set FaceMesh(FaceMesh:Skeletal Mesh Component Budgeted Object Reference=Get Array Item with EFort Custom Part Type()) -> (Output_Get:Skeletal Mesh Component Budgeted Object Reference)
   > AddTickPrerequisiteComponent!(execute, self:Actor Component Object Reference=GetOwningComponent(), PrerequisiteComponent:Actor Component Object Reference=@n25) -> (then)
   Event OnBeginSkydiving()
   > Set TrailControllerRot(TrailControllerRot="0,-90,0")
   Event OnEndTacticalSprinting() {is=title K2Node_Event}
   > Set iswalking(iswalking:double=1) -> (Output_Get:double)
   Event OnEndSkydiving()
   > Set TrailControllerRot(TrailControllerRot="0,-35,0")
   Event OnBeginTacticalSprinting() {is=title K2Node_Event}
   > Set iswalking(iswalking=0)
   float + float!(A=1, B=MapRangeClamped(Value=0, InRangeA=0, InRangeB=1)) -> (ReturnValue)
  ]}
]
variables [
 angularSpringConstant {type=number}
 Body {type=boolean}
 BwdMovement {type=number}
 FaceMesh {type=SkeletalMeshComponentBudgeted}
 Gravity Override {type=Vector}
 Is Wearing Backpack {type=boolean}
 Is Wearing Cape {type=boolean}
 isCrouching {type=number}
 ismoving2D {type=number}
 isShovelCrop_Cape {type=boolean}
 iswalking {type=number}
 iswearingbackbling {type=number}
 moveBackward {type=number}
 moveForward {type=number}
 RBAN Settings Out {type=FortRigidBodyAnimNodeParameters}
 strafeLeft float {type=number}
 strafeRight float {type=number}
 TrailControllerRot {type=Rotator}
]
